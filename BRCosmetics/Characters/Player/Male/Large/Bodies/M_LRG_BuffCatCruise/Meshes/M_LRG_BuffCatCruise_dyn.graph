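# Physics asset for the large male BuffCatCruise player body, setting up physics for its shirt bones.
# Five capsule chains (left and right sides, three back strips) link through swing- and twist-limited constraints to spine_02, except the right side, which anchors to spine_03.
# Pelvis, spine and thigh-twist capsules plus glute spheres supply body shapes.
# A two-capsule dyn_gun body has no constraint.
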PhysicsAsset /BRCosmetics/Characters/Player/Male/Large/Bodies/M_LRG_BuffCatCruise/Meshes/M_LRG_BuffCatCruise_dyn
bodies {
 deform_glute_l [{shapeType=Sphere, radius=12, shapeName=deform_glute_l_sphere, center=<9.9073 3.3022 6.746>}]
 deform_glute_r [{shapeType=Sphere, radius=12, shapeName=deform_glute_l_sphere, center=<-9.9073 -3.3022 -6.746>}]
 dyn_gun [
  {
   shapeType=Capsule
   length=17.75
   radius=3.75
   shapeName=dyn_gun_capsule
   center=<6.4889 -3.4596 2.8624>
   rotation=<rot 82.4328 169.2596 164.3244>
  }
  {
   shapeType=Capsule
   length=11.75
   radius=3.75
   shapeName=dyn_gun_capsule
   center=<-3.3946 3.2684 2.0585>
   rotation=<rot 82.1462 -152.854 -68.1071>
  }
 ]
 dyn_shirt_bk_01 [{shapeType=Capsule, length=15, radius=7.5, shapeName=dyn_shirt_bk_01_capsule, center=<5.7806 6.5057 0>}]
 dyn_shirt_bk_02 [{shapeType=Capsule, length=15, radius=7.5, shapeName=dyn_shirt_bk_02_capsule, center=<9.7563 7.1263 0>}]
 dyn_shirt_bk_l_01 [{
   shapeType=Capsule
   length=16.75
   radius=7.5
   shapeName=dyn_shirt_bk_l_01_capsule
   center=<-2.7259 -5.862 -2.1036>
   rotation=<rot 0.0011 -0.001 -11.2499>
  }]
 dyn_shirt_bk_l_02 [{
   shapeType=Capsule
   length=16.75
   radius=7.5
   shapeName=dyn_shirt_bk_l_02_capsule
   center=<-5.0302 -5.8508 -0.4667>
   rotation=<rot 0.0011 -0.001 -11.2499>
  }]
 dyn_shirt_bk_r_01 [{
   shapeType=Capsule
   length=15
   radius=7.5
   shapeName=dyn_shirt_bk_r_01_capsule
   center=<6.3926 4.8126 -3.3376>
   rotation=<rot 0.0012 -0.0023 22.8128>
  }]
 dyn_shirt_bk_r_02 [{
   shapeType=Capsule
   length=15
   radius=7.5
   shapeName=dyn_shirt_bk_r_02_capsule
   center=<9.0236 3.1467 -3.6877>
   rotation=<rot 0.0011 -0.0021 14.3752>
  }]
 dyn_shirt_l_01 [{
   shapeType=Capsule
   length=25.25
   radius=7.5
   shapeName=dyn_shirt_l_01_capsule
   center=<-9.0664 -5.2516 -4.2492>
   rotation=<rot 0.0023 0 -5.9375>
  }]
 dyn_shirt_l_02 [{shapeType=Capsule, length=25.75, radius=8, shapeName=dyn_shirt_l_02_capsule, center=<-9.7462 -6.2561 -4.5172>, rotation=<rot 0.0033 0 -12>}]
 dyn_shirt_l_03 [{
   shapeType=Capsule
   length=25.25
   radius=7.5
   shapeName=dyn_shirt_l_03_capsule
   center=<-10.3021 -8.3279 -3.7552>
   rotation=<rot 0.001 0 -6.5625>
  }]
 dyn_shirt_r_01 [{
   shapeType=Capsule
   length=15
   radius=7.5
   shapeName=dyn_shirt_r_01_capsule
   center=<7.3722 6.2206 5.2208>
   rotation=<rot 0.0014 -0.0003 -25.3122>
  }]
 dyn_shirt_r_02 [{
   shapeType=Capsule
   length=15
   radius=7.5
   shapeName=dyn_shirt_r_02_capsule
   center=<9.7426 6.8652 1.7759>
   rotation=<rot 0.0009 -0.0002 -16.8748>
  }]
 dyn_shirt_r_03 [{
   shapeType=Capsule
   length=15
   radius=7.5
   shapeName=dyn_shirt_r_03_capsule
   center=<8.7824 5.1647 4.9758>
   rotation=<rot 0.0005 -0.0001 -11.2499>
  }]
 pelvis [{shapeType=Capsule, length=1, radius=16.25, shapeName=pelvis_capsule, center=<0 0.1292 -0.254>}]
 spine_01 [{shapeType=Capsule, length=5.5, radius=13.75, shapeName=spine_01_capsule, center=<0 -0.8855 -0.5696>}]
 spine_02 [{shapeType=Capsule, length=5.75, radius=14.25, shapeName=spine_02_capsule, center=<0 0.4745 0.7225>}]
 spine_03 [{shapeType=Capsule, length=8.25, radius=14, shapeName=spine_03_capsule, center=<0 0 1.9741>}]
 thigh_twist_01_l [{shapeType=Capsule, length=21.5917, radius=9.7951, center=<4.046 0.8859 2.5652>, rotation=<rot -0.2707 -89.2388 83.9888>}]
 thigh_twist_01_r [{shapeType=Capsule, length=17.5917, radius=8.7951, center=<-8.0684 -0.9374 -2.9887>, rotation=<rot -0.2707 -89.2388 -96.0112>}]
}
constraints [
 {bone1Name=dyn_shirt_r_03, bone2Name=dyn_shirt_r_02, swing1LimitDegrees=30, swing2LimitDegrees=5, twistLimitDegrees=5}
 {bone1Name=dyn_shirt_r_02, bone2Name=dyn_shirt_r_01, swing1LimitDegrees=30, swing2LimitDegrees=5, twistLimitDegrees=5}
 {bone1Name=dyn_shirt_r_01, bone2Name=spine_03, swing1LimitDegrees=30, swing2LimitDegrees=5, twistLimitDegrees=5}
 {bone1Name=dyn_shirt_bk_l_02, bone2Name=dyn_shirt_bk_l_01, swing1LimitDegrees=45, swing2LimitDegrees=5, twistLimitDegrees=3}
 {bone1Name=dyn_shirt_bk_l_01, bone2Name=spine_02, swing1LimitDegrees=45, swing2LimitDegrees=5, twistLimitDegrees=3}
 {bone1Name=dyn_shirt_bk_02, bone2Name=dyn_shirt_bk_01, swing1LimitDegrees=45, swing2LimitDegrees=20, twistLimitDegrees=10}
 {bone1Name=dyn_shirt_bk_01, bone2Name=spine_02, swing1LimitDegrees=45, swing2LimitDegrees=20, twistLimitDegrees=10}
 {bone1Name=dyn_shirt_bk_r_02, bone2Name=dyn_shirt_bk_r_01, swing1LimitDegrees=45, swing2LimitDegrees=5, twistLimitDegrees=10}
 {bone1Name=dyn_shirt_bk_r_01, bone2Name=spine_02, swing1LimitDegrees=45, swing2LimitDegrees=10, twistLimitDegrees=10}
 {bone1Name=dyn_shirt_l_03, bone2Name=dyn_shirt_l_02, swing1LimitDegrees=30, swing2LimitDegrees=5, twistLimitDegrees=20}
 {bone1Name=dyn_shirt_l_02, bone2Name=dyn_shirt_l_01, swing1LimitDegrees=60, swing2LimitDegrees=5, twistLimitDegrees=20}
 {bone1Name=dyn_shirt_l_01, bone2Name=spine_02, swing1LimitDegrees=60, swing2LimitDegrees=5, twistLimitDegrees=15}
]
