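# Animation Blueprint for the BuffCatCruise body cosmetic, adding secondary motion to its shirt and gun bones.
# Copied parent pose runs through a constraint, deform Control Rig, pelvis-space RigidBody with state-based gravity, and gun AnimDynamics.
# Bone offsets then correct clipping-prone poses using movement, crouch, slide, melee and skydive states.
# Spine twist position history yields a velocity-driven shirt rotation alpha, damped by pawn speed and zeroed when airborne.
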AnimBlueprint /BRCosmetics/Characters/Player/Male/Large/Bodies/M_LRG_BuffCatCruise/Meshes/M_LRG_BuffCatCruise_AnimBP
target_skeleton=&/M_LRG_BuffCatCruise_Skeleton
graphs [
 AnimGraph {nodes [
   n1 = Get Back shirt Z rot() -> (Back shirt Z rot:double)
   n2 = Get Shirt corr() -> (Shirt corr:bool)
   n3 = Get Back shirt Y rot() -> (Back shirt Y rot:double)
   n4 = "Transform (Modify) Bone - Bone: dyn_shirt_bk_l_01"!(
    ComponentPose=Local To Component(LocalPose=Use cached pose 'History corr pose'())
    bAlphaBoolEnabled:bool=@n2
    Translation="0,0,0"
    Rotation_Roll:float=0
    Rotation_Pitch:float=@n3
    Rotation_Yaw:float=@n1
    Scale="1,1,1"
   ) -> (Pose) {
    alphaInputType=Bool
    rotationMode=BMM_Additive
    rotationSpace=BCS_ParentBoneSpace
    alphaBoolBlend {blendInTime=0.5, blendOption=Linear, blendOutTime=0.3}
    boneToModify {boneName=dyn_shirt_bk_l_01}
   }
   n5 = "Transform (Modify) Bone - Bone: dyn_shirt_bk_01"!(
    ComponentPose=@n4
    bAlphaBoolEnabled:bool=@n2
    Translation="0,0,0"
    Rotation_Roll:float=0
    Rotation_Pitch:float=@n3
    Rotation_Yaw:float=@n1
    Scale="1,1,1"
   ) -> (Pose) {
    alphaInputType=Bool
    rotationMode=BMM_Additive
    rotationSpace=BCS_ParentBoneSpace
    alphaBoolBlend {blendInTime=0.5, blendOption=Linear, blendOutTime=0.3}
    boneToModify {boneName=dyn_shirt_bk_01}
   }
   n6 = "Transform (Modify) Bone - Bone: dyn_shirt_bk_r_01"!(
    ComponentPose=@n5
    bAlphaBoolEnabled:bool=@n2
    Translation="0,0,0"
    Rotation_Roll:float=0
    Rotation_Pitch:float=@n3
    Rotation_Yaw:float=@n1
    Scale="1,1,1"
   ) -> (Pose) {
    alphaInputType=Bool
    rotationMode=BMM_Additive
    rotationSpace=BCS_ParentBoneSpace
    alphaBoolBlend {blendInTime=0.5, blendOption=Linear, blendOutTime=0.3}
    boneToModify {boneName=dyn_shirt_bk_r_01}
   }
   n7 = "Transform (Modify) Bone - Bone: dyn_gun"!(
    ComponentPose=@n6
    bAlphaBoolEnabled:bool=True
    Translation="0,0,0"
    Rotation_Roll:float=0
    Rotation_Pitch:float=Get Gun Rot() -> (Gun Rot:double)
    Rotation_Yaw:float=0
    Scale="1,1,1"
   ) -> (Pose) {
    alphaInputType=Bool
    rotationMode=BMM_Additive
    rotationSpace=BCS_BoneSpace
    alphaBoolBlend {blendOption=Linear}
    boneToModify {boneName=dyn_gun}
    rotation {pitch=1.5}
   }
   Output Pose(Result=Component To Local(ComponentPose=@n7)) {layerGroup=None}
   n8 = "Constraint - dyn_shirt_r_02 (P)"(ComponentPose=Local To Component(LocalPose=Copy Pose From Mesh() {bUseAttachedParent=true}), Alpha=1, ConstraintWeights_0=1) {
    alphaBoolBlend {blendOption=Linear}
    boneToModify {boneName=dyn_shirt_r_02}
    constraintSetup [{targetBone {boneName=dyn_gun}}]
    constraintWeights [1]
   }
   n9 = Control Rig(Source=Component To Local(ComponentPose=@n8)) {
    alphaBoolBlend {blendOption=Linear}
    controlRigAssetReference {blueprintRigClass=&/M_LRG_BuffCatCruise_DeformRig_Master_ControlRig.M_LRG_BuffCatCruise_DeformRig_Master_ControlRig_C}
   }
   n10 = RigidBody!(ComponentPose=Local To Component(LocalPose=@n9), Alpha, OverrideWorldGravity:Vector=Get Gravity Override(), ExternalForce) -> (Pose) {
    bClampLinearTranslationLimitToRefPose=true
    bOverrideWorldGravity=true
    overridePhysicsAsset=&/M_LRG_BuffCatCruise_dyn
    simulationSpace=BaseBoneSpace
    alphaBoolBlend {blendOption=Linear}
    baseBoneRef {boneName=pelvis}
    componentAppliedLinearAccClamp=<3000 3000 3000>
    componentLinearAccScale=<0.3 0.3 0.5>
    componentLinearVelScale=<0.3 0.3 0.5>
   }
   n11 = Anim Dynamics - Bone: dyn_gun(ComponentPose=@n10, Alpha=1, ExternalForce="0,0,0") {
    angularDampingOverride=0.92
    angularSpringConstant=290
    bAngularSpring=true
    bOverrideAngularDamping=true
    bUseGravityOverride=true
    bUseSphericalLimits=true
    simulationSpace=RootRelative
    alphaBoolBlend {blendOption=Linear}
    boundBone {boneName=dyn_gun}
    componentAppliedLinearAccClamp=<3000 3000 3000>
    componentLinearAccScale=<0.1 0.1 0.1>
    componentLinearVelScale=<0.1 0.1 0.1>
    gravityOverride {z=-3000}
    physicsBodyDefinitions [{
      boundBone {boneName=dyn_gun}
      boxExtents=<18 30 10>
      constraintSetup {angularLimitsMax=<10 0 10>, angularLimitsMin=<0 -10 -10>, angularTarget=<1 0 0>}
      localJointOffset=<2 0 4.5>
     }]
    relativeSpaceBone {boneName=spine_01}
    sphericalLimits [{limitRadius=7.32, drivingBone {boneName=thigh_r}, sphereLocalOffset=<0 0 0>}]
   }
   n12 = "Transform (Modify) Bone - Bone: dyn_shirt_bk_01"!(ComponentPose=@n11, bAlphaBoolEnabled=Get bIsSkyDiving(), Translation="0,0,0", Rotation:Rotator="0,0,0", Scale="0.8,1,1") -> (Pose) {
    alphaInputType=Bool
    scaleMode=BMM_Replace
    scaleSpace=BCS_BoneSpace
    translationSpace=BCS_BoneSpace
    alphaBoolBlend {blendOption=Linear}
    boneToModify {boneName=dyn_shirt_bk_01}
    scale {x=0.8}
   }
   n13 = "Transform (Modify) Bone - Bone: dyn_gun"!(ComponentPose=@n12, bAlphaBoolEnabled=Get Gun corr() -> (Gun corr:bool), Translation="0,-1,0", Rotation:Rotator="-5,0,0", Scale="1,1,1") -> (Pose) {
    alphaInputType=Bool
    rotationMode=BMM_Additive
    scaleMode=BMM_Replace
    scaleSpace=BCS_BoneSpace
    translationMode=BMM_Additive
    translationSpace=BCS_BoneSpace
    alphaBoolBlend {blendOption=Linear}
    boneToModify {boneName=dyn_gun}
    rotation {pitch=-5}
    translation {y=-1}
   }
   n14 = "Transform (Modify) Bone - Bone: dyn_shirt_r_02"!(
    ComponentPose=@n13
    bAlphaBoolEnabled:bool=Get Moving forward() -> (Moving forward:bool)
    Translation="0,0,0"
    Rotation="0.0,8,0.0"
    Scale="1,1,1"
   ) -> (Pose) {
    alphaInputType=Bool
    rotationMode=BMM_Additive
    rotationSpace=BCS_BoneSpace
    alphaBoolBlend {blendOption=Linear}
    boneToModify {boneName=dyn_shirt_r_02}
   }
   Save cached pose 'SavedPose1958'(Pose=Component To Local(ComponentPose=@n14)) {cacheName=SavedPose1958}
   n15 = Get AlphaRotationHistory() -> (AlphaRotationHistory:float)
   n16 = "Transform (Modify) Bone - Bone: dyn_shirt_bk_l_01"!(
    ComponentPose=Local To Component(LocalPose=Use cached pose 'SavedPose1958'())
    Alpha:float=@n15
    Translation="0,0,0"
    Rotation:Rotator="0,20,0"
    Scale="1,1,1"
   ) -> (Pose) {
    rotationMode=BMM_Additive
    rotationSpace=BCS_BoneSpace
    alphaBoolBlend {blendOption=Linear}
    alphaScaleBiasClamp {bInterpResult=true, interpSpeedDecreasing=5}
    boneToModify {boneName=dyn_shirt_bk_l_01}
    rotation {yaw=20}
   }
   n17 = "Transform (Modify) Bone - Bone: dyn_shirt_bk_01"!(ComponentPose=@n16, Alpha:float=@n15, Translation="0,0,0", Rotation="0,25,0", Scale="1,1,1") -> (Pose) {
    rotationMode=BMM_Additive
    rotationSpace=BCS_BoneSpace
    alphaBoolBlend {blendOption=Linear}
    alphaScaleBiasClamp {bInterpResult=true, interpSpeedDecreasing=5}
    boneToModify {boneName=dyn_shirt_bk_01}
    rotation {yaw=25}
   }
   n18 = "Transform (Modify) Bone - Bone: dyn_shirt_bk_r_01"!(ComponentPose=@n17, Alpha:float=@n15, Translation="0,0,0", Rotation:Rotator="0,20,0", Scale="1,1,1") -> (Pose) {
    rotationMode=BMM_Additive
    rotationSpace=BCS_BoneSpace
    alphaBoolBlend {blendOption=Linear}
    alphaScaleBiasClamp {bInterpResult=true, interpSpeedDecreasing=5}
    boneToModify {boneName=dyn_shirt_bk_r_01}
    rotation {yaw=20}
   }
   n19 = "Transform (Modify) Bone - Bone: dyn_shirt_r_01"(ComponentPose=@n18, Alpha=@n15, Translation="0,0,0", Rotation="0,20,0", Scale="1,1,1") {
    rotationMode=BMM_Additive
    rotationSpace=BCS_BoneSpace
    alphaBoolBlend {blendOption=Linear}
    alphaScaleBiasClamp {bInterpResult=true, interpSpeedDecreasing=5}
    boneToModify {boneName=dyn_shirt_r_01}
    rotation {yaw=20}
   }
   n20 = "Transform (Modify) Bone - Bone: dyn_shirt_l_01"(ComponentPose=@n19, Alpha=@n15, Translation="0,0,0", Rotation="0,20,0", Scale="1,1,1") {
    rotationMode=BMM_Additive
    rotationSpace=BCS_BoneSpace
    alphaBoolBlend {blendOption=Linear}
    alphaScaleBiasClamp {bInterpResult=true, interpSpeedDecreasing=5}
    boneToModify {boneName=dyn_shirt_l_01}
    rotation {yaw=20}
   }
   Save cached pose 'History corr pose'(Pose=Component To Local(ComponentPose=@n20)) {cacheName=History corr pose}
  ]}
 EventGraph {nodes [
   n1 = Event BlueprintUpdateAnimation()
   > Set Delta Time X(Delta Time X=@n1.DeltaTimeX)
   > n2 = Sequence()
   > n3 = Gravity Override State Function DEPRECATED(CharacterPartAnimInstance=Self-Reference(), GameTimeInSeconds=GetGameTimeInSeconds(), GravityOverrideParams=Get Gravity Override Params())
   > Set Gravity Override(Gravity Override=@n3)
   n4 = Get PawnMovementDirectionAzimuth() -> (PawnMovementDirectionAzimuth:float)
   > Set Moving left(Moving left:bool=AND Boolean!(A=Get bIsOnGround() -> (bIsOnGround:bool), B=Get bIsMoving2D(), C:bool="float > float"(A=@n4, B=-190), D:bool="float < float"(A=@n4, B=-135)) -> (ReturnValue)) -> (Output_Get:bool)
   > Set Moving forward(Moving forward:bool=AND Boolean!(A="float > float"(A=@n4, B=-45), B="float < float"(A=@n4, B=45)) -> (ReturnValue)) -> (Output_Get:bool)
   > Set Moving Right(Moving Right:bool=AND Boolean(A="float > float"(A=@n4, B=45), B="float < float"(A=@n4, B=135))) -> (Output_Get:bool)
   n5 = Get Moving forward()
   n6 = Get Moving Right() -> (Moving Right:bool)
   n7 = Get Moving left() -> (Moving left:bool)
   > Set Gun corr(Gun corr:bool=OR Boolean(A=OR Boolean!(A=@n7, B=@n5, C:bool=@n6) -> (ReturnValue), B=Get bIsCrouching())) -> (Output_Get:bool)
   n8 = Get SkeletalMesh() -> (SkeletalMesh:Skeletal Mesh Component Object Reference)
   > Set Spine_05_Rotation(Spine_05_Rotation:Vector=MakeVector(Y=GetAlphaFromReferenceTransformDelta(SkeletalMeshComponent=@n8, BoneName=spine_05, BaseName=spine_01, DeltaComponentType=EulerY, InputRange=(X=0,Y=1), TwistAxis="1,0,0").RawOutput)) -> (Output_Get:Vector)
   n9 = Calculate Velocity From Position History!(
    execute=@n2.then_1
    DeltaSeconds:float=Get Delta Time X()
    Position:Vector=Get Spine_05_Rotation() -> (Spine_05_Rotation:Vector)
    History:"Position History Structure (by ref)"=Get History() -> (History:Position History Structure)
    NumberOfSamples:int=16
    VelocityMin:float=5
    VelocityMax:float=64
   ) -> (then, ReturnValue:float)
   > Set Shirt corr(Shirt corr:bool=AND Boolean!(A=Get bIsCrouching(), B=Get bIsPlayingMeleeAnim(), C:bool=Get bIsMoving2D()) -> (ReturnValue)) -> (Output_Get:bool)
   n10 = OR Boolean!(A=Get bIsPlayingMeleeAnim(), B=Get bIsSkyDiving()) -> (ReturnValue)
   n11 = MapRangeClamped(Value=Vector Length(A=Get PawnVelocity()), InRangeA=0, InRangeB=700, OutRangeA=1, OutRangeB=0)
   n12 = Get bIsJumping()
   n13 = Get bIsFalling()
   n14 = "To Float (Boolean)"(InBool=NOT Boolean(A=OR Boolean(A=@n12, B=@n13)))
   > Set AlphaRotationHistory(AlphaRotationHistory:float="Clamp (Float)"(
     Value=float * float!(A=@n9, B=@n11, C:double=@n14) -> (ReturnValue) {is=title K2Node_PromotableOperator}
     Min=0
     Max=SelectFloat(A=0.4, B=1, bPickA=@n10)
    )) -> (Output_Get:float)
   n15 = BreakVector(InVec=Get Spine_05_Rotation())
   > Set Back shirt Y rot(Back shirt Y rot:double=float / float(A=@n15.Y, B=1)) -> (Output_Get:double)
   > Set Back shirt Z rot(Back shirt Z rot:double=float / float(A=@n15.Y, B=-2)) -> (Output_Get:double)
   n16 = Get bIsMoving2D()
   > n17 = Set sits in place(sits in place:bool=AND Boolean!(A=NOT Boolean(A=@n16), B=Get bIsCrouching()) -> (ReturnValue)) -> (Output_Get:bool)
   > Set Gun Rot(Gun Rot:double=SelectFloat(A=30, B=SelectFloat(A=5, B=1.5, bPickA=@n17), bPickA=Get bIsSliding())) -> (Output_Get:double)
   Event OnCharacterCustomizationLoaded()
   > Set SkeletalMesh(SkeletalMesh:Skeletal Mesh Component Object Reference=GetOwnerSkeletalMeshComponent()) -> (Output_Get:Skeletal Mesh Component Object Reference)
  ]}
]
variables [
 AlphaRotationHistory {type=number}
 Back shirt Y rot {type=number}
 Back shirt Z rot {type=number}
 Delta Time X {type=number}
 Gravity Override {type=Vector}
 Gravity Override Params {
  type=GravityOverrideParamsStruct
  default {
   onGround_Moving {gravityOverride=<0 0 -1100>}
   onGround_Standing {gravityOverride=<0 0 -1100>}
   skydive_DivingDown {windAmplitude=<150 1100 300>, windFrequency=<2 23 22>}
   skydive_FloatingDown {windAmplitude=<300 -300 300>}
   skydive_GoingUp {gravityOverride=<0 -700 600>, windAmplitude=<300 -300 300>, windOffset=<0 -600 400>}
  }
 }
 Gun corr {type=boolean}
 Gun Rot {type=number}
 History {type=PositionHistory, default {range=0, positions []}}
 Moving forward {type=boolean}
 Moving left {type=boolean}
 Moving Right {type=boolean}
 Shirt corr {type=boolean}
 sits in place {type=boolean}
 SkeletalMesh {type=SkeletalMeshComponent}
 Spine_05_Rotation {type=Vector}
]
