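# Physics asset for the large male BuffCatCruise body, named as the collider that capes collide against.
# Capsules cover torso, neck, limbs, hands and feet, with a sphere for the head and small capsules on the dyn_shirt back and side bone chains.
# Constraints link the bodies along the skeleton with all swing and twist free.
# The shirt chains attach to spine_02, except dyn_shirt_r_01, which attaches to spine_03.
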PhysicsAsset /BRCosmetics/Characters/Player/Male/Large/Bodies/M_LRG_BuffCatCruise/Meshes/M_LRG_BuffCatCruise_CapeCollider
bodies {
 calf_l [{shapeType=Capsule, length=27, radius=15, shapeName=calf_l_capsule, center=<-18.0265 0.0001 0>, rotation=<rot 90 160.5288 160.529>}]
 calf_r [{shapeType=Capsule, length=27, radius=15, shapeName=calf_r_capsule, center=<17.9789 -0.0001 0>, rotation=<rot 90 160.5288 160.529>}]
 clavicle_l [{shapeType=Capsule, length=15, radius=8.5, shapeName=clavicle_l_capsule, center=<7.8129 0 4.6521>, rotation=<rot -80 179.9989 -179.999>}]
 clavicle_r [{shapeType=Capsule, length=15, radius=8.5, shapeName=clavicle_r_capsule, center=<-9.8859 0 -4.5319>, rotation=<rot -80 179.9989 -179.9989>}]
 dyn_shirt_bk_01 [{shapeType=Capsule, length=8.5, radius=6, shapeName=dyn_shirt_bk_01_capsule, rotation=<rot 90 90 90.0002>}]
 dyn_shirt_bk_02 [{shapeType=Capsule, length=8.5, radius=6, shapeName=dyn_shirt_bk_02_capsule, rotation=<rot 90 86.1042 86.1043>}]
 dyn_shirt_bk_l_01 [{shapeType=Capsule, length=8.5, radius=6, shapeName=dyn_shirt_bk_l_01_capsule, rotation=<rot 90 173.3637 173.3638>}]
 dyn_shirt_bk_l_02 [{shapeType=Capsule, length=8.5, radius=6, shapeName=dyn_shirt_bk_l_02_capsule, rotation=<rot 90 175.0684 175.0685>}]
 dyn_shirt_bk_r_01 [{shapeType=Capsule, length=8.5, radius=6, shapeName=dyn_shirt_bk_r_01_capsule, rotation=<rot 90 167.9934 167.9935>}]
 dyn_shirt_bk_r_02 [{shapeType=Capsule, length=8.5, radius=6, shapeName=dyn_shirt_bk_r_02_capsule, rotation=<rot 90 169.6981 169.6982>}]
 dyn_shirt_l_01 [{shapeType=Capsule, length=8.5, radius=6, shapeName=dyn_shirt_l_01_capsule, rotation=<rot 90 179.6278 179.6278>}]
 dyn_shirt_l_02 [{shapeType=Capsule, length=8.5, radius=6, shapeName=dyn_shirt_l_02_capsule, rotation=<rot 90 -178.6675 -178.6675>}]
 dyn_shirt_l_03 [{shapeType=Capsule, length=8.5, radius=6, shapeName=dyn_shirt_l_03_capsule, rotation=<rot 90 179.8248 179.8248>}]
 dyn_shirt_r_01 [{shapeType=Capsule, length=8.5, radius=6, shapeName=dyn_shirt_r_01_capsule, rotation=<rot 90 167.4895 167.4896>}]
 dyn_shirt_r_02 [{shapeType=Capsule, length=8.5, radius=6, shapeName=dyn_shirt_r_02_capsule, rotation=<rot 90 169.1942 169.1943>}]
 dyn_shirt_r_03 [{shapeType=Capsule, length=8.5, radius=6, shapeName=dyn_shirt_r_03_capsule, rotation=<rot 90 167.6864 167.6865>}]
 foot_l [{shapeType=Capsule, length=15, radius=8.5, shapeName=foot_l_capsule, center=<-6.5073 7.5993 0.7544>, rotation=<rot 75.894 -134.5617 -45.4383>}]
 foot_r [{
   shapeType=Capsule
   length=10.5
   radius=9
   shapeName=foot_r_capsule
   center=<7.1979 -7.0315 -1.1398>
   rotation=<rot 75.8941 -134.5618 -45.4381>
  }]
 hand_l [{shapeType=Capsule, length=11, radius=6.5, shapeName=hand_l_capsule, center=<6.7781 -1.0794 0>, rotation=<rot 90 -179.9603 -179.9607>}]
 hand_r [{shapeType=Capsule, length=10.5, radius=7, shapeName=hand_r_capsule, center=<-5.1681 2.1126 0>, rotation=<rot -90 0 0>}]
 head [{shapeType=Sphere, radius=16, shapeName=head_sphere, center=<10.0448 0 0>}]
 lowerarm_l [{shapeType=Capsule, length=18, radius=8, shapeName=lowerarm_l_capsule, center=<6.1765 0 0>, rotation=<rot 90 -0.0007 -0.001>}]
 lowerarm_r [{shapeType=Capsule, length=18, radius=8, shapeName=lowerarm_r_capsule, center=<-5.2744 0 0>, rotation=<rot 90 -160.5289 -160.5293>}]
 neck_02 [{shapeType=Capsule, length=4, radius=6, shapeName=neck_02_capsule, rotation=<rot -90 -13.5122 13.5121>}]
 pelvis [{shapeType=Capsule, length=25, radius=18, shapeName=pelvis_capsule, center=<-7.5341 0 0>}]
 spine_01 [{shapeType=Capsule, length=25.5, radius=16, shapeName=spine_01_capsule}]
 spine_02 [{shapeType=Capsule, length=15, radius=18.5, shapeName=spine_02_capsule}]
 spine_03 [{shapeType=Capsule, length=18.5, radius=16.5, shapeName=spine_03_capsule}]
 spine_04 [{shapeType=Capsule, length=15, radius=15, shapeName=spine_04_capsule}]
 spine_05 [{shapeType=Capsule, length=15, radius=21, shapeName=spine_05_capsule}]
 thigh_l [{shapeType=Capsule, length=16.5, radius=16, shapeName=thigh_l_capsule, center=<-21.7647 0.0002 0>, rotation=<rot 90 3.0357 3.0362>}]
 thigh_r [{shapeType=Capsule, length=16.5, radius=16, shapeName=thigh_r_capsule, center=<22.3523 -0.0002 0>, rotation=<rot 90 179.9602 179.9608>}]
 upperarm_l [{shapeType=Capsule, length=18, radius=8.5, shapeName=upperarm_l_capsule, center=<3.8841 0 0>, rotation=<rot 80 -0.0002 -0.0008>}]
 upperarm_r [{shapeType=Capsule, length=18, radius=8.5, shapeName=upperarm_r_capsule, center=<-2.3175 0 0>, rotation=<rot 80 -0.0002 -0.0008>}]
}
constraints [
 {
  bone1Name=dyn_shirt_bk_02
  bone2Name=dyn_shirt_bk_01
  swing1LimitDegrees=45
  swing1Motion=Free
  swing2LimitDegrees=45
  swing2Motion=Free
  twistLimitDegrees=45
  twistMotion=Free
 }
 {
  bone1Name=dyn_shirt_bk_l_02
  bone2Name=dyn_shirt_bk_l_01
  swing1LimitDegrees=45
  swing1Motion=Free
  swing2LimitDegrees=45
  swing2Motion=Free
  twistLimitDegrees=45
  twistMotion=Free
 }
 {
  bone1Name=dyn_shirt_bk_r_02
  bone2Name=dyn_shirt_bk_r_01
  swing1LimitDegrees=45
  swing1Motion=Free
  swing2LimitDegrees=45
  swing2Motion=Free
  twistLimitDegrees=45
  twistMotion=Free
 }
 {
  bone1Name=dyn_shirt_l_02
  bone2Name=dyn_shirt_l_01
  swing1LimitDegrees=45
  swing1Motion=Free
  swing2LimitDegrees=45
  swing2Motion=Free
  twistLimitDegrees=45
  twistMotion=Free
 }
 {
  bone1Name=dyn_shirt_l_03
  bone2Name=dyn_shirt_l_02
  swing1LimitDegrees=45
  swing1Motion=Free
  swing2LimitDegrees=45
  swing2Motion=Free
  twistLimitDegrees=45
  twistMotion=Free
 }
 {
  bone1Name=dyn_shirt_r_02
  bone2Name=dyn_shirt_r_01
  swing1LimitDegrees=45
  swing1Motion=Free
  swing2LimitDegrees=45
  swing2Motion=Free
  twistLimitDegrees=45
  twistMotion=Free
 }
 {
  bone1Name=dyn_shirt_r_03
  bone2Name=dyn_shirt_r_02
  swing1LimitDegrees=45
  swing1Motion=Free
  swing2LimitDegrees=45
  swing2Motion=Free
  twistLimitDegrees=45
  twistMotion=Free
 }
 {
  bone1Name=hand_l
  bone2Name=lowerarm_l
  swing1LimitDegrees=45
  swing1Motion=Free
  swing2LimitDegrees=45
  swing2Motion=Free
  twistLimitDegrees=45
  twistMotion=Free
 }
 {
  bone1Name=calf_l
  bone2Name=thigh_l
  swing1LimitDegrees=45
  swing1Motion=Free
  swing2LimitDegrees=45
  swing2Motion=Free
  twistLimitDegrees=45
  twistMotion=Free
 }
 {
  bone1Name=foot_l
  bone2Name=calf_l
  swing1LimitDegrees=45
  swing1Motion=Free
  swing2LimitDegrees=45
  swing2Motion=Free
  twistLimitDegrees=45
  twistMotion=Free
 }
 {
  bone1Name=calf_r
  bone2Name=thigh_r
  swing1LimitDegrees=45
  swing1Motion=Free
  swing2LimitDegrees=45
  swing2Motion=Free
  twistLimitDegrees=45
  twistMotion=Free
 }
 {
  bone1Name=foot_r
  bone2Name=calf_r
  swing1LimitDegrees=45
  swing1Motion=Free
  swing2LimitDegrees=45
  swing2Motion=Free
  twistLimitDegrees=45
  twistMotion=Free
 }
 {
  bone1Name=thigh_r
  bone2Name=pelvis
  swing1LimitDegrees=45
  swing1Motion=Free
  swing2LimitDegrees=45
  swing2Motion=Free
  twistLimitDegrees=45
  twistMotion=Free
 }
 {
  bone1Name=thigh_l
  bone2Name=pelvis
  swing1LimitDegrees=45
  swing1Motion=Free
  swing2LimitDegrees=45
  swing2Motion=Free
  twistLimitDegrees=45
  twistMotion=Free
 }
 {
  bone1Name=dyn_shirt_r_01
  bone2Name=spine_03
  swing1LimitDegrees=45
  swing1Motion=Free
  swing2LimitDegrees=45
  swing2Motion=Free
  twistLimitDegrees=45
  twistMotion=Free
 }
 {
  bone1Name=spine_01
  bone2Name=pelvis
  swing1LimitDegrees=45
  swing1Motion=Free
  swing2LimitDegrees=45
  swing2Motion=Free
  twistLimitDegrees=45
  twistMotion=Free
 }
 {
  bone1Name=spine_02
  bone2Name=spine_01
  swing1LimitDegrees=45
  swing1Motion=Free
  swing2LimitDegrees=45
  swing2Motion=Free
  twistLimitDegrees=45
  twistMotion=Free
 }
 {
  bone1Name=spine_03
  bone2Name=spine_02
  swing1LimitDegrees=45
  swing1Motion=Free
  swing2LimitDegrees=45
  swing2Motion=Free
  twistLimitDegrees=45
  twistMotion=Free
 }
 {
  bone1Name=spine_04
  bone2Name=spine_03
  swing1LimitDegrees=45
  swing1Motion=Free
  swing2LimitDegrees=45
  swing2Motion=Free
  twistLimitDegrees=45
  twistMotion=Free
 }
 {
  bone1Name=spine_05
  bone2Name=spine_04
  swing1LimitDegrees=45
  swing1Motion=Free
  swing2LimitDegrees=45
  swing2Motion=Free
  twistLimitDegrees=45
  twistMotion=Free
 }
 {
  bone1Name=neck_02
  bone2Name=head
  swing1LimitDegrees=45
  swing1Motion=Free
  swing2LimitDegrees=45
  swing2Motion=Free
  twistLimitDegrees=45
  twistMotion=Free
 }
 {
  bone1Name=spine_05
  bone2Name=neck_02
  swing1LimitDegrees=45
  swing1Motion=Free
  swing2LimitDegrees=45
  swing2Motion=Free
  twistLimitDegrees=45
  twistMotion=Free
 }
 {
  bone1Name=dyn_shirt_bk_01
  bone2Name=spine_02
  swing1LimitDegrees=45
  swing1Motion=Free
  swing2LimitDegrees=45
  swing2Motion=Free
  twistLimitDegrees=45
  twistMotion=Free
 }
 {
  bone1Name=dyn_shirt_bk_l_01
  bone2Name=spine_02
  swing1LimitDegrees=45
  swing1Motion=Free
  swing2LimitDegrees=45
  swing2Motion=Free
  twistLimitDegrees=45
  twistMotion=Free
 }
 {
  bone1Name=dyn_shirt_bk_r_01
  bone2Name=spine_02
  swing1LimitDegrees=45
  swing1Motion=Free
  swing2LimitDegrees=45
  swing2Motion=Free
  twistLimitDegrees=45
  twistMotion=Free
 }
 {
  bone1Name=dyn_shirt_l_01
  bone2Name=spine_02
  swing1LimitDegrees=45
  swing1Motion=Free
  swing2LimitDegrees=45
  swing2Motion=Free
  twistLimitDegrees=45
  twistMotion=Free
 }
 {
  bone1Name=clavicle_l
  bone2Name=spine_05
  swing1LimitDegrees=45
  swing1Motion=Free
  swing2LimitDegrees=45
  swing2Motion=Free
  twistLimitDegrees=45
  twistMotion=Free
 }
 {
  bone1Name=clavicle_r
  bone2Name=spine_05
  swing1LimitDegrees=45
  swing1Motion=Free
  swing2LimitDegrees=45
  swing2Motion=Free
  twistLimitDegrees=45
  twistMotion=Free
 }
 {
  bone1Name=upperarm_r
  bone2Name=clavicle_r
  swing1LimitDegrees=45
  swing1Motion=Free
  swing2LimitDegrees=45
  swing2Motion=Free
  twistLimitDegrees=45
  twistMotion=Free
 }
 {
  bone1Name=upperarm_l
  bone2Name=clavicle_l
  swing1LimitDegrees=45
  swing1Motion=Free
  swing2LimitDegrees=45
  swing2Motion=Free
  twistLimitDegrees=45
  twistMotion=Free
 }
 {
  bone1Name=lowerarm_r
  bone2Name=upperarm_r
  swing1LimitDegrees=45
  swing1Motion=Free
  swing2LimitDegrees=45
  swing2Motion=Free
  twistLimitDegrees=45
  twistMotion=Free
 }
 {
  bone1Name=hand_r
  bone2Name=lowerarm_r
  swing1LimitDegrees=45
  swing1Motion=Free
  swing2LimitDegrees=45
  swing2Motion=Free
  twistLimitDegrees=45
  twistMotion=Free
 }
 {
  bone1Name=lowerarm_l
  bone2Name=upperarm_l
  swing1LimitDegrees=45
  swing1Motion=Free
  swing2LimitDegrees=45
  swing2Motion=Free
  twistLimitDegrees=45
  twistMotion=Free
 }
]
