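# Control Rig that poses the long back hair of the BuffCatCruise large male character against head and body motion.
# Center nulls, constrained to head, neck and spine, drive a B-spline IK and twist chain, using controls gathered at construction.
# Left and right strands follow the center chain and clavicles, with spherical pose readers on each upper arm weighting corrective offsets.
# Separate side spline IK chains are left unconnected to execution.
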ControlRigBlueprint /BRCosmetics/Characters/Player/Male/Large/Bodies/M_LRG_BuffCatCruise/Meshes/Parts/M_LRG_BuffCatCruise_FaceAcc_HairCtrlRig
graphs [
 ContainedGraph {nodes [
   n1 = RigVMFunctionEntryNode()
   SequenceAggregate = &RigVMFunction_Sequence(ExecuteContext=@n1.ExecuteContext)
   SequenceAggregate_1 = &RigVMFunction_Sequence(ExecuteContext=@SequenceAggregate.B)
   SequenceAggregate_2 = &RigVMFunction_Sequence(ExecuteContext=@SequenceAggregate_1.B)
   RigVMFunctionReturnNode(A=@SequenceAggregate.A, B=@SequenceAggregate_1.A, C=@SequenceAggregate_2.A, D=@SequenceAggregate_2.B)
  ]}
 ContainedGraph {nodes [
   n1 = RigVMFunctionEntryNode()
   SequenceAggregate = &RigVMFunction_Sequence(ExecuteContext=@n1.ExecuteContext)
   SequenceAggregate_1 = &RigVMFunction_Sequence(ExecuteContext=@SequenceAggregate.B)
   SequenceAggregate_2 = &RigVMFunction_Sequence(ExecuteContext=@SequenceAggregate_1.B)
   RigVMFunctionReturnNode(A=@SequenceAggregate.A, B=@SequenceAggregate_1.A, C=@SequenceAggregate_2.A, D=@SequenceAggregate_2.B)
  ]}
 RigVMFunctionLibrary {links [], nodes []}
 RigVMModel {nodes [
   &RigUnit_ParentConstraint(
    AdvancedSettings=(RotationOrderForFilter=XZY)
    Child=(Type=Null,Name="dyn_Hair_BackL_01_Null")
    ExecutePin=()
    Filter=(TranslationFilter=(bX=true,bY=true,bZ=true),RotationFilter=(bX=false,bY=false,bZ=false),ScaleFilter=(bX=True,bY=True,bZ=True))
    Parents=((Item=(Type=Bone,Name="head"),Weight=0.9),(Item=(Type=Bone,Name="spine_05"),Weight=0.1))
    Weight=1
   )
   > &RigUnit_ParentConstraint(
    AdvancedSettings=(RotationOrderForFilter=XZY)
    Child=(Type=Null,Name="dyn_Hair_BackL_02_Null")
    Filter=(TranslationFilter=(bX=True,bY=True,bZ=True),RotationFilter=(bX=false,bY=false,bZ=false),ScaleFilter=(bX=True,bY=True,bZ=True))
    Parents=((Item=(Type=Bone,Name="head"),Weight=0.6),(Item=(Type=Bone,Name="spine_05"),Weight=0.4))
    Weight=1
   )
   > &RigUnit_ParentConstraint(
    AdvancedSettings=(RotationOrderForFilter=XZY)
    Child=(Type=Null,Name="dyn_Hair_BackL_03_Null")
    Filter=(TranslationFilter=(bX=True,bY=True,bZ=True),RotationFilter=(bX=false,bY=false,bZ=false),ScaleFilter=(bX=True,bY=True,bZ=True))
    Parents=((Item=(Type=Bone,Name="dyn_Hair_BackL_03"),Weight=0.5),(Item=(Type=Bone,Name="spine_05"),Weight=0.5))
    Weight=1
   )
   > &RigUnit_ParentConstraint(
    AdvancedSettings=(RotationOrderForFilter=XZY)
    Child=(Type=Null,Name="dyn_Hair_BackL_04_Null")
    Filter=(TranslationFilter=(bX=True,bY=True,bZ=True),RotationFilter=(bX=false,bY=false,bZ=false),ScaleFilter=(bX=True,bY=True,bZ=True))
    Parents=((Item=(Type=Bone,Name="clavicle_l"),Weight=0),(Item=(Type=Bone,Name="spine_05"),Weight=1))
    Weight=1
   )
   > &RigUnit_ParentConstraint(
    AdvancedSettings=(RotationOrderForFilter=XZY)
    Child=(Type=Null,Name="dyn_Hair_BackL_05_Null")
    Filter=(TranslationFilter=(bX=True,bY=True,bZ=True),RotationFilter=(bX=false,bY=false,bZ=false),ScaleFilter=(bX=True,bY=True,bZ=True))
    Parents=((Item=(Type=Bone,Name="clavicle_l"),Weight=0.15),(Item=(Type=Bone,Name="spine_04"),Weight=0.85))
    Weight=1
   )
   > &RigUnit_ParentConstraint(
    AdvancedSettings=(RotationOrderForFilter=XZY)
    Child=(Type=Null,Name="dyn_Hair_BackL_06_Null")
    Filter=(TranslationFilter=(bX=true,bY=true,bZ=true),RotationFilter=(bX=true,bY=true,bZ=true),ScaleFilter=(bX=True,bY=True,bZ=True))
    Parents=((Item=(Type=Bone,Name="dyn_Hair_BackL_06"),Weight=0.2),(Item=(Type=Bone,Name="spine_02"),Weight=0.8))
    Weight=1
   )
   RerouteNode = RigVMDispatchNode(Values=((Type=Bone,Name="dyn_Hair_BackL_01"),((),Type=Bone,Name="dyn_Hair_BackL_02"),((),Type=Bone,Name="dyn_Hair_BackL_03"),((),Type=Bone,Name="dyn_Hair_BackL_04"),((),Type=Bone,Name="dyn_Hair_BackL_05"),((),Type=Bone,Name="dyn_Hair_BackL_06"))) {template="DISPATCH_RigVMDispatch_ArrayMake(in Values,out Array)"}
   > SplineIK = RigVMFunctionReferenceNode(
    Primary Axis=(X=0,Y=-1,Z=0)
    Samples Per Segment=1
    Secondary Spline Direction=(X=1,Y=1,Z=1)
    Spline Mode=BSpline
    Stretch=false
    Up Axis=(X=0,Y=0,Z=0)
    debug=false
    Bones=@RerouteNode.Array
    Controls=RigVMVariableNode() {variable=dyn_Hair_BackL_ctrls}.Value
   )
   > TwistBones_1 = RigVMFunctionReferenceNode(
    Axis to Twist=(bX=false,bY=true,bZ=false)
    Bones to Twist=((Type=Bone,Name="dyn_Hair_BackL_01"),(Type=Bone,Name="dyn_Hair_BackL_05"))
    End=(Type=Bone,Name="head")
    Start=(Type=Bone,Name="spine_02")
   )
   &RigUnit_ParentConstraint(
    AdvancedSettings=(RotationOrderForFilter=XZY)
    Child=(Type=Null,Name="dyn_Hair_BackR_01_Null")
    ExecutePin=()
    Filter=(TranslationFilter=(bX=true,bY=true,bZ=true),RotationFilter=(bX=false,bY=false,bZ=false),ScaleFilter=(bX=True,bY=True,bZ=True))
    Parents=((Item=(Type=Bone,Name="head"),Weight=0.8),(Item=(Type=Bone,Name="spine_05"),Weight=0.2))
    Weight=1
   )
   > &RigUnit_ParentConstraint(
    AdvancedSettings=(RotationOrderForFilter=XZY)
    Child=(Type=Null,Name="dyn_Hair_BackR_02_Null")
    Filter=(TranslationFilter=(bX=True,bY=True,bZ=True),RotationFilter=(bX=false,bY=false,bZ=false),ScaleFilter=(bX=True,bY=True,bZ=True))
    Parents=((Item=(Type=Bone,Name="neck_01"),Weight=0.8),(Item=(Type=Bone,Name="spine_05"),Weight=0.2))
    Weight=1
   )
   > &RigUnit_ParentConstraint(
    AdvancedSettings=(RotationOrderForFilter=XZY)
    Child=(Type=Null,Name="dyn_Hair_BackR_03_Null")
    Filter=(TranslationFilter=(bX=True,bY=True,bZ=True),RotationFilter=(bX=false,bY=false,bZ=false),ScaleFilter=(bX=True,bY=True,bZ=True))
    Parents=((Item=(Type=Bone,Name="dyn_Hair_BackR_03"),Weight=0.4),(Item=(Type=Bone,Name="spine_05"),Weight=0.6))
    Weight=1
   )
   > &RigUnit_ParentConstraint(
    AdvancedSettings=(RotationOrderForFilter=XZY)
    Child=(Type=Null,Name="dyn_Hair_BackR_04_Null")
    Filter=(TranslationFilter=(bX=True,bY=True,bZ=True),RotationFilter=(bX=false,bY=false,bZ=false),ScaleFilter=(bX=True,bY=True,bZ=True))
    Parents=((Item=(Type=Bone,Name="upperarm_r"),Weight=0.5),(Item=(Type=Bone,Name="spine_05"),Weight=0.5))
    Weight=1
   )
   > &RigUnit_ParentConstraint(
    AdvancedSettings=(RotationOrderForFilter=XZY)
    Child=(Type=Null,Name="dyn_Hair_BackR_05_Null")
    Filter=(TranslationFilter=(bX=True,bY=True,bZ=True),RotationFilter=(bX=false,bY=false,bZ=false),ScaleFilter=(bX=True,bY=True,bZ=True))
    Parents=((Item=(Type=Bone,Name="clavicle_r"),Weight=0),(Item=(Type=Bone,Name="spine_04"),Weight=1))
    Weight=1
   )
   > &RigUnit_ParentConstraint(
    AdvancedSettings=(RotationOrderForFilter=XZY)
    Child=(Type=Null,Name="dyn_Hair_BackR_06_Null")
    Filter=(TranslationFilter=(bX=True,bY=True,bZ=True),RotationFilter=(bX=true,bY=true,bZ=true),ScaleFilter=(bX=True,bY=True,bZ=True))
    Parents=((Item=(Type=Bone,Name="dyn_Hair_BackR_06"),Weight=0.2),(Item=(Type=Bone,Name="spine_02"),Weight=0.8))
    Weight=1
   )
   RerouteNode_1 = RigVMDispatchNode(Values=((Type=Bone,Name="dyn_Hair_BackR_01"),((),Type=Bone,Name="dyn_Hair_BackR_02"),((),Type=Bone,Name="dyn_Hair_BackR_03"),((),Type=Bone,Name="dyn_Hair_BackR_04"),((),Type=Bone,Name="dyn_Hair_BackR_05"),((),Type=Bone,Name="dyn_Hair_BackR_06"))) {template="DISPATCH_RigVMDispatch_ArrayMake(in Values,out Array)"}
   > SplineIK_1 = RigVMFunctionReferenceNode(
    Primary Axis=(X=0,Y=-1,Z=0)
    Samples Per Segment=1
    Secondary Spline Direction=(X=1,Y=1,Z=1)
    Spline Mode=BSpline
    Stretch=false
    Up Axis=(X=0,Y=0,Z=0)
    debug=false
    Bones=@RerouteNode_1.Array
    Controls=RigVMVariableNode() {variable=dyn_Hair_BackR_ctrls}.Value
   )
   RigUnit_BeginExecution = &RigUnit_BeginExecution()
   SequenceAggregate = RigVMAggregateNode(ExecuteContext=@RigUnit_BeginExecution.ExecutePin)
   &RigUnit_ParentConstraint(
    AdvancedSettings=(RotationOrderForFilter=XZY)
    Child=(Type=Null,Name="dyn_Hair_Back_01_Null")
    Filter=(TranslationFilter=(bX=true,bY=true,bZ=true),RotationFilter=(bX=false,bY=false,bZ=false),ScaleFilter=(bX=True,bY=True,bZ=True))
    Parents=((Item=(Type=Bone,Name="head"),Weight=0.9),(Item=(Type=Bone,Name="spine_05"),Weight=0.1))
    Weight=1
    ExecutePin=@SequenceAggregate.A
   )
   > &RigUnit_ParentConstraint(
    AdvancedSettings=(RotationOrderForFilter=XZY)
    Child=(Type=Null,Name="dyn_Hair_Back_02_Null")
    Filter=(TranslationFilter=(bX=True,bY=True,bZ=True),RotationFilter=(bX=false,bY=false,bZ=false),ScaleFilter=(bX=True,bY=True,bZ=True))
    Parents=((Item=(Type=Bone,Name="neck_02"),Weight=0.8),(Item=(Type=Bone,Name="spine_05"),Weight=0.2))
    Weight=1
   )
   > &RigUnit_ParentConstraint(
    AdvancedSettings=(RotationOrderForFilter=XZY)
    Child=(Type=Null,Name="dyn_Hair_Back_03_Null")
    Filter=(TranslationFilter=(bX=True,bY=True,bZ=True),RotationFilter=(bX=false,bY=false,bZ=false),ScaleFilter=(bX=True,bY=True,bZ=True))
    Parents=((Item=(Type=Bone,Name="neck_02"),Weight=0.7),(Item=(Type=Bone,Name="spine_05"),Weight=0.3))
    Weight=1
   )
   > &RigUnit_ParentConstraint(
    AdvancedSettings=(RotationOrderForFilter=XZY)
    Child=(Type=Null,Name="dyn_Hair_Back_04_Null")
    Filter=(TranslationFilter=(bX=True,bY=True,bZ=True),RotationFilter=(bX=false,bY=false,bZ=false),ScaleFilter=(bX=True,bY=True,bZ=True))
    Parents=((Item=(Type=Bone,Name="clavicle_r"),Weight=0),(Item=(Type=Bone,Name="spine_05"),Weight=1),(Item=(Type=Bone,Name="clavicle_l"),Weight=0))
    Weight=1
   )
   > &RigUnit_ParentConstraint(
    AdvancedSettings=(RotationOrderForFilter=XZY)
    Child=(Type=Null,Name="dyn_Hair_Back_05_Null")
    Filter=(TranslationFilter=(bX=true,bY=true,bZ=true),RotationFilter=(bX=false,bY=false,bZ=false),ScaleFilter=(bX=True,bY=True,bZ=True))
    Parents=((Item=(Type=Bone,Name="clavicle_l"),Weight=0),(Item=(Type=Bone,Name="spine_05"),Weight=1),(Item=(Type=Bone,Name="clavicle_r"),Weight=0))
    Weight=1
   )
   > &RigUnit_ParentConstraint(
    AdvancedSettings=(RotationOrderForFilter=XZY)
    Child=(Type=Null,Name="dyn_Hair_Back_06_Null")
    Filter=(TranslationFilter=(bX=true,bY=true,bZ=true),RotationFilter=(bX=false,bY=false,bZ=false),ScaleFilter=(bX=True,bY=True,bZ=True))
    Parents=((Item=(Type=Bone,Name="dyn_Hair_Back_06"),Weight=0),(Item=(Type=Bone,Name="spine_05"),Weight=1))
    Weight=1
   )
   RerouteNode_2 = RigVMDispatchNode(Values=((Type=Bone,Name="dyn_Hair_Back_01"),((),Type=Bone,Name="dyn_Hair_Back_02"),((),Type=Bone,Name="dyn_Hair_Back_03"),((),Type=Bone,Name="dyn_Hair_Back_04"),((),Type=Bone,Name="dyn_Hair_Back_05"),((),Type=Bone,Name="dyn_Hair_Back_06"))) {template="DISPATCH_RigVMDispatch_ArrayMake(in Values,out Array)"}
   > SplineIK_2 = RigVMFunctionReferenceNode(
    Primary Axis=(X=1,Y=-2,Z=0)
    Samples Per Segment=1
    Secondary Spline Direction=(X=0,Y=0,Z=0)
    Spline Mode=BSpline
    Stretch=false
    Up Axis=(X=0,Y=0,Z=0)
    debug=false
    Bones=@RerouteNode_2.Array
    Controls=RigVMVariableNode() {variable=dyn_Hair_Back_ctrls}.Value
   )
   > TwistBones = RigVMFunctionReferenceNode(
    Axis to Twist=(bX=true,bY=true,bZ=false)
    Bones to Twist=((Type=Bone,Name="dyn_Hair_Back_03"),(Type=Bone,Name="dyn_Hair_Back_06"))
    End=(Type=Bone,Name="spine_05")
    Start=(Type=Bone,Name="spine_05")
   )
   &RigUnit_ParentConstraint(
    AdvancedSettings=(RotationOrderForFilter=XZY)
    Child=(Type=Bone,Name="dyn_Hair_BackL_01")
    Filter=(TranslationFilter=(bX=true,bY=true,bZ=true),RotationFilter=(bX=true,bY=true,bZ=true),ScaleFilter=(bX=True,bY=True,bZ=True))
    Parents=((Item=(Type=Bone,Name="dyn_Hair_Back_01"),Weight=1),(Item=(Type=Bone,Name="spine_05"),Weight=0))
    Weight=1
    ExecutePin=@SequenceAggregate.B
   )
   > &RigUnit_ParentConstraint(
    AdvancedSettings=(RotationOrderForFilter=XZY)
    Child=(Type=Bone,Name="dyn_Hair_BackL_02")
    Filter=(TranslationFilter=(bX=True,bY=True,bZ=True),RotationFilter=(bX=true,bY=true,bZ=true),ScaleFilter=(bX=True,bY=True,bZ=True))
    Parents=((Item=(Type=Bone,Name="dyn_Hair_Back_02"),Weight=0.5),(Item=(Type=Bone,Name="clavicle_l"),Weight=0.5))
    Weight=1
   )
   > &RigUnit_ParentConstraint(
    AdvancedSettings=(RotationOrderForFilter=XZY)
    Child=(Type=Bone,Name="dyn_Hair_BackL_03")
    Filter=(TranslationFilter=(bX=True,bY=True,bZ=True),RotationFilter=(bX=true,bY=true,bZ=true),ScaleFilter=(bX=True,bY=True,bZ=True))
    Parents=((Item=(Type=Bone,Name="dyn_Hair_Back_03"),Weight=0.7),(Item=(Type=Bone,Name="clavicle_l"),Weight=0.3))
    Weight=1
   )
   > &RigUnit_ParentConstraint(
    AdvancedSettings=(RotationOrderForFilter=XZY)
    Child=(Type=Bone,Name="dyn_Hair_BackL_04")
    Filter=(TranslationFilter=(bX=True,bY=True,bZ=True),RotationFilter=(bX=true,bY=true,bZ=true),ScaleFilter=(bX=True,bY=True,bZ=True))
    Parents=((Item=(Type=Bone,Name="dyn_Hair_Back_04"),Weight=1),(Item=(Type=Bone,Name="spine_05"),Weight=0))
    Weight=1
   )
   > &RigUnit_ParentConstraint(
    AdvancedSettings=(RotationOrderForFilter=XZY)
    Child=(Type=Bone,Name="dyn_Hair_BackL_05")
    Filter=(TranslationFilter=(bX=True,bY=True,bZ=True),RotationFilter=(bX=true,bY=true,bZ=true),ScaleFilter=(bX=True,bY=True,bZ=True))
    Parents=((Item=(Type=Bone,Name="dyn_Hair_Back_05"),Weight=1),(Item=(Type=Bone,Name="spine_04"),Weight=0))
    Weight=1
   )
   > &RigUnit_ParentConstraint(
    AdvancedSettings=(RotationOrderForFilter=XZY)
    Child=(Type=Bone,Name="dyn_Hair_BackL_06")
    Filter=(TranslationFilter=(bX=true,bY=true,bZ=true),RotationFilter=(bX=true,bY=true,bZ=true),ScaleFilter=(bX=True,bY=True,bZ=True))
    Parents=((Item=(Type=Bone,Name="dyn_Hair_Back_06"),Weight=1),(Item=(Type=Bone,Name="spine_02"),Weight=0))
    Weight=1
   )
   > n1 = &RigUnit_SphericalPoseReader(
    ActiveRegionScaleFactors=(PositiveWidth=0.6,NegativeWidth=0.6,PositiveHeight=0.6,NegativeHeight=0.6)
    ActiveRegionSize=0.1
    Debug=(bDrawDebug=false,DebugScale=25,DebugSegments=20,DebugThickness=0.25)
    DriverAxis=(X=1,Y=0,Z=0)
    DriverItem=(Type=Bone,Name="upperarm_l")
    FalloffRegionScaleFactors=(PositiveWidth=1,NegativeWidth=0.6,PositiveHeight=0.6,NegativeHeight=0.6)
    FalloffSize=0.55
    FlipHeightScaling=false
    FlipWidthScaling=false
    OptionalParentItem=(Type=Bone,Name="spine_05")
    RotationOffset=(X=-80,Y=-30,Z=-10)
   )
   > &RigUnit_ParentConstraint(
    AdvancedSettings=(RotationOrderForFilter=XZY)
    Child=(Type=Bone,Name="dyn_Hair_BackL_03")
    Filter=(TranslationFilter=(bX=false,bY=false,bZ=false),RotationFilter=(bX=true,bY=true,bZ=true),ScaleFilter=(bX=True,bY=True,bZ=True))
    Parents=((Item=(Type=Bone,Name="dyn_Hair_Back_03"),Weight=0.3),(Item=(Type=Bone,Name="upperarm_l"),Weight=0.7))
    Weight=@n1.OutputParam
   )
   n2 = &RigVMFunction_MathQuaternionFromEuler(Euler=(X=20,Y=-10,Z=0), RotationOrder=YXZ)
   > &RigUnit_ModifyTransforms(
    ItemToModify="((Item=(Type=Bone,Name=\"dyn_Hair_BackL_02\"),Transform=(Rotation=(X=0,Y=0,Z=0,W=1),Translation=(X=2,Y=-1,Z=1),Scale3D=(X=1,Y=1,Z=1))),(Item=(Type=Bone,Name=\"dyn_Hair_BackL_04\"),Transform=(Rotation=(X=0,Y=0,Z=0,W=1),Translation=(X=0,Y=0,Z=0),Scale3D=(X=1,Y=1 ...(truncated)"
    ItemToModify.0.Transform.Rotation=@n2.Result
    ItemToModify.1.Transform.Rotation=@n2.Result
    ItemToModify.2.Transform.Rotation=&RigVMFunction_MathQuaternionFromEuler(Euler=(X=0,Y=-20,Z=25)).Result
    Weight=@n1.OutputParam
   )
   n3 = &RigVMFunction_MathQuaternionFromEuler(Euler=(X=0,Y=0,Z=5))
   > &RigUnit_ModifyTransforms(
    ItemToModify="((Item=(Type=Bone,Name=\"dyn_Hair_Back_03\"),Transform=(Rotation=(X=0,Y=0,Z=0,W=1),Translation=(X=0,Y=0,Z=1),Scale3D=(X=1,Y=1,Z=1))),(Item=(Type=Bone,Name=\"dyn_Hair_Back_04\"),Transform=(Rotation=(X=0,Y=0,Z=0,W=1),Translation=(X=0,Y=0,Z=0),Scale3D=(X=1,Y=1,Z= ...(truncated)"
    ItemToModify.2.Transform.Rotation=&RigVMFunction_MathQuaternionFromEuler(Euler=(X=0,Y=0,Z=2)).Result
    ItemToModify.0.Transform.Rotation=@n3.Result
    ItemToModify.1.Transform.Rotation=@n3.Result
    Weight=@n1.OutputParam
   )
   &RigUnit_ParentConstraint(
    AdvancedSettings=(RotationOrderForFilter=XZY)
    Child=(Type=Bone,Name="dyn_Hair_BackR_01")
    Filter=(TranslationFilter=(bX=true,bY=true,bZ=true),RotationFilter=(bX=true,bY=true,bZ=true),ScaleFilter=(bX=True,bY=True,bZ=True))
    Parents=((Item=(Type=Bone,Name="dyn_Hair_Back_01"),Weight=1),(Item=(Type=Bone,Name="spine_05"),Weight=0))
    Weight=1
    ExecutePin=@SequenceAggregate.C
   )
   > &RigUnit_ParentConstraint(
    AdvancedSettings=(RotationOrderForFilter=XZY)
    Child=(Type=Bone,Name="dyn_Hair_BackR_02")
    Filter=(TranslationFilter=(bX=True,bY=True,bZ=True),RotationFilter=(bX=true,bY=true,bZ=true),ScaleFilter=(bX=True,bY=True,bZ=True))
    Parents=((Item=(Type=Bone,Name="dyn_Hair_Back_02"),Weight=0.5),(Item=(Type=Bone,Name="clavicle_r"),Weight=0.5))
    Weight=1
   )
   > &RigUnit_ParentConstraint(
    AdvancedSettings=(RotationOrderForFilter=XZY)
    Child=(Type=Bone,Name="dyn_Hair_BackR_03")
    Filter=(TranslationFilter=(bX=True,bY=True,bZ=True),RotationFilter=(bX=true,bY=true,bZ=true),ScaleFilter=(bX=True,bY=True,bZ=True))
    Parents=((Item=(Type=Bone,Name="dyn_Hair_Back_03"),Weight=0.5),(Item=(Type=Bone,Name="clavicle_r"),Weight=0.5))
    Weight=1
   )
   > &RigUnit_ParentConstraint(
    AdvancedSettings=(RotationOrderForFilter=XZY)
    Child=(Type=Bone,Name="dyn_Hair_BackR_04")
    Filter=(TranslationFilter=(bX=True,bY=True,bZ=True),RotationFilter=(bX=true,bY=true,bZ=true),ScaleFilter=(bX=True,bY=True,bZ=True))
    Parents=((Item=(Type=Bone,Name="dyn_Hair_Back_04"),Weight=1),(Item=(Type=Bone,Name="spine_05"),Weight=0))
    Weight=1
   )
   > &RigUnit_ParentConstraint(
    AdvancedSettings=(RotationOrderForFilter=XZY)
    Child=(Type=Bone,Name="dyn_Hair_BackR_05")
    Filter=(TranslationFilter=(bX=True,bY=True,bZ=True),RotationFilter=(bX=true,bY=true,bZ=true),ScaleFilter=(bX=True,bY=True,bZ=True))
    Parents=((Item=(Type=Bone,Name="dyn_Hair_Back_05"),Weight=1),(Item=(Type=Bone,Name="spine_04"),Weight=0))
    Weight=1
   )
   > &RigUnit_ParentConstraint(
    AdvancedSettings=(RotationOrderForFilter=XZY)
    Child=(Type=Bone,Name="dyn_Hair_BackR_06")
    Filter=(TranslationFilter=(bX=true,bY=true,bZ=true),RotationFilter=(bX=true,bY=true,bZ=true),ScaleFilter=(bX=True,bY=True,bZ=True))
    Parents=((Item=(Type=Bone,Name="dyn_Hair_Back_06"),Weight=1),(Item=(Type=Bone,Name="spine_02"),Weight=0))
    Weight=1
   )
   > n4 = &RigUnit_SphericalPoseReader(
    ActiveRegionScaleFactors=(PositiveWidth=0.6,NegativeWidth=0.6,PositiveHeight=0.6,NegativeHeight=0.6)
    ActiveRegionSize=0.1
    Debug=(bDrawDebug=false,DebugScale=25,DebugSegments=20,DebugThickness=0.25)
    DriverAxis=(X=-1,Y=0,Z=0)
    DriverItem=(Type=Bone,Name="upperarm_r")
    FalloffRegionScaleFactors=(PositiveWidth=1,NegativeWidth=0.6,PositiveHeight=0.6,NegativeHeight=0.6)
    FalloffSize=0.6
    FlipHeightScaling=false
    FlipWidthScaling=true
    OptionalParentItem=(Type=Bone,Name="spine_05")
    RotationOffset=(X=80,Y=-30,Z=-10)
   )
   > &RigUnit_ParentConstraint(
    AdvancedSettings=(RotationOrderForFilter=XZY)
    Child=(Type=Bone,Name="dyn_Hair_BackR_03")
    Filter=(TranslationFilter=(bX=false,bY=false,bZ=false),RotationFilter=(bX=true,bY=true,bZ=true),ScaleFilter=(bX=True,bY=True,bZ=True))
    Parents=((Item=(Type=Bone,Name="dyn_Hair_Back_03"),Weight=0.3),(Item=(Type=Bone,Name="upperarm_r"),Weight=0.7))
    Weight=@n4.OutputParam
   )
   n5 = &RigVMFunction_MathQuaternionFromEuler(Euler=(X=-20,Y=10,Z=0), RotationOrder=YXZ)
   > &RigUnit_ModifyTransforms(
    ItemToModify="((Item=(Type=Bone,Name=\"dyn_Hair_BackR_02\"),Transform=(Rotation=(X=0,Y=0,Z=0,W=1),Translation=(X=2,Y=-1,Z=1),Scale3D=(X=1,Y=1,Z=1))),(Item=(Type=Bone,Name=\"dyn_Hair_BackR_04\"),Transform=(Rotation=(X=0,Y=0,Z=0,W=1),Translation=(X=0,Y=0,Z=0),Scale3D=(X=1,Y=1 ...(truncated)"
    ItemToModify.0.Transform.Rotation=@n5.Result
    ItemToModify.1.Transform.Rotation=@n5.Result
    ItemToModify.2.Transform.Rotation=&RigVMFunction_MathQuaternionFromEuler(Euler=(X=0,Y=20,Z=25)).Result
    Weight=@n4.OutputParam
   )
   n6 = &RigVMFunction_MathQuaternionFromEuler(Euler=(X=0,Y=0,Z=5))
   > n7 = &RigUnit_ModifyTransforms(
    ItemToModify="((Item=(Type=Bone,Name=\"dyn_Hair_Back_03\"),Transform=(Rotation=(X=0,Y=0,Z=0,W=1),Translation=(X=0,Y=0,Z=1),Scale3D=(X=1,Y=1,Z=1))),(Item=(Type=Bone,Name=\"dyn_Hair_Back_04\"),Transform=(Rotation=(X=0,Y=0,Z=0,W=1),Translation=(X=0,Y=0,Z=0),Scale3D=(X=1,Y=1,Z= ...(truncated)"
    ItemToModify.2.Transform.Rotation=&RigVMFunction_MathQuaternionFromEuler(Euler=(X=0,Y=0,Z=2)).Result
    ItemToModify.0.Transform.Rotation=@n6.Result
    ItemToModify.1.Transform.Rotation=@n6.Result
    Weight=@n4.OutputParam
   )
   n8 = &RigUnit_SphericalPoseReader(
    ActiveRegionScaleFactors=(PositiveWidth=1,NegativeWidth=1,PositiveHeight=1,NegativeHeight=1)
    ActiveRegionSize=0.2
    Debug=(bDrawDebug=false,DebugScale=25,DebugSegments=20,DebugThickness=0.25)
    DriverAxis=(X=1,Y=0,Z=0)
    DriverItem=(Type=Bone,Name="head")
    ExecutePin=()
    FalloffRegionScaleFactors=(PositiveWidth=1,NegativeWidth=1,PositiveHeight=1,NegativeHeight=1)
    FalloffSize=0.7
    FlipHeightScaling=false
    FlipWidthScaling=true
    OptionalParentItem=(Type=Bone,Name="clavicle_r")
    RotationOffset=(X=-80,Y=0,Z=7)
   )
   &RigUnit_ParentConstraint(
    AdvancedSettings=(RotationOrderForFilter=XZY)
    Child=(Type=Bone,Name="dyn_hair_fringe_r_02")
    Filter=(TranslationFilter=(bX=true,bY=true,bZ=true),RotationFilter=(bX=true,bY=true,bZ=true),ScaleFilter=(bX=True,bY=True,bZ=True))
    Parents=((Item=(Type=Bone,Name="neck_02"),Weight=0.6),(Item=(Type=Bone,Name="clavicle_r"),Weight=0.4))
    ExecutePin=@n7.ExecutePin
    Weight=@n8.OutputParam
   )
   n9 = &RigUnit_SphericalPoseReader(
    ActiveRegionScaleFactors=(PositiveWidth=1,NegativeWidth=1,PositiveHeight=1,NegativeHeight=1)
    ActiveRegionSize=0.2
    Debug=(bDrawDebug=false,DebugScale=25,DebugSegments=20,DebugThickness=0.25)
    DriverAxis=(X=1,Y=0,Z=0)
    DriverItem=(Type=Bone,Name="head")
    ExecutePin=()
    FalloffRegionScaleFactors=(PositiveWidth=1,NegativeWidth=1,PositiveHeight=1,NegativeHeight=1)
    FalloffSize=0.7
    FlipHeightScaling=false
    FlipWidthScaling=true
    OptionalParentItem=(Type=Bone,Name="clavicle_l")
    RotationOffset=(X=-100,Y=0,Z=7)
   )
   > &RigUnit_ParentConstraint(
    AdvancedSettings=(RotationOrderForFilter=XZY)
    Child=(Type=Bone,Name="dyn_Hair_BackL_03a")
    Filter=(TranslationFilter=(bX=false,bY=false,bZ=false),RotationFilter=(bX=true,bY=true,bZ=true),ScaleFilter=(bX=True,bY=True,bZ=True))
    Parents=((Item=(Type=Bone,Name="dyn_Hair_BackL_03"),Weight=0.3),(Item=(Type=Bone,Name="clavicle_l"),Weight=0.7))
    Weight=@n9.OutputParam
   )
   > &RigUnit_ModifyTransforms(
    ItemToModify=((Item=(Type=Bone,Name="dyn_Hair_BackL_03a"),Transform=(Rotation=(X=0,Y=0,Z=0,W=1),Translation=(X=1,Y=-1,Z=2),Scale3D=(X=1,Y=1,Z=1))))
    ItemToModify.0.Transform.Rotation=&RigVMFunction_MathQuaternionFromEuler(Euler=(X=20,Y=-30,Z=20)).Result
    Weight=@n9.OutputParam
   )
   &RigVMFunction_MathQuaternionFromEuler(Euler=(X=0,Y=20,Z=10))
   &RigUnit_ModifyTransforms(
    ExecutePin=()
    ItemToModify=((Item=(Type=Bone,Name="dyn_hair_fringe_r_03"),Transform=(Rotation=(X=0,Y=0,Z=0,W=1),Translation=(X=1,Y=1,Z=-2),Scale3D=(X=1,Y=1,Z=1))),(Item=(Type=Bone,Name="None"),Transform=(Rotation=(X=0,Y=0,Z=0,W=1),Translation=(X=0,Y=0,Z=0),Scale3D=(X=1,Y=1,Z=1))))
    ItemToModify.0.Transform.Rotation=&RigVMFunction_MathQuaternionFromEuler(Euler=(X=-20,Y=10,Z=20)).Result
    Weight=@n8.OutputParam
   )
   RerouteNode_2_1 = RigVMDispatchNode(Values=((Type=Bone,Name="dyn_Hair_Back_01"),((),Type=Bone,Name="dyn_Hair_Back_02"),((),Type=Bone,Name="dyn_Hair_Back_03"),((),Type=Bone,Name="dyn_Hair_Back_04"),((),Type=Bone,Name="dyn_Hair_Back_05"),((),Type=Bone,Name="dyn_Hair_Back_06"))) {template="DISPATCH_RigVMDispatch_ArrayMake(in Values,out Array)"}
   TwistBones_1_1 = RigVMFunctionReferenceNode(
    Axis to Twist=(bX=true,bY=false,bZ=false)
    Bones to Twist=((Type=Bone,Name="dyn_Hair_BackR_01"),(Type=Bone,Name="dyn_Hair_BackR_06"))
    End=(Type=Bone,Name="spine_05")
    ExecuteContext=()
    Start=(Type=Bone,Name="spine_02")
   )
  ]}
 RigVMModel Construction Event Graph {nodes [
   n1 = &RigUnit_PrepareForExecution()
   SequenceAggregate = RigVMAggregateNode(ExecuteContext=@n1.ExecutePin)
   RerouteNode_6_1 = RigVMDispatchNode(Values="((Type=Control,Name=\"dyn_Hair_Back_01_ctrl\"),((),Type=Control,Name=\"dyn_Hair_Back_02_ctrl\"),((),Type=Control,Name=\"dyn_Hair_Back_03_ctrl\"),((),Type=Control,Name=\"dyn_Hair_Back_04_ctrl\"),((),Type=Control,Name=\"dyn_Hair_Back_05_ctrl\"),((),Type=Control,Name=\" ...(truncated)") {template="DISPATCH_RigVMDispatch_ArrayMake(in Values,out Array)"}
   RigVMVariableNode(Value=@RerouteNode_6_1.Array, ExecuteContext=@SequenceAggregate.A) {variable=dyn_Hair_Back_ctrls}
   > RigVMVariableNode(Value=/Script/ControlRig.RigUnit_HierarchyGetPoseItemArray(ElementType=All, Initial=false, ItemsToGet=@RerouteNode_6_1.Array) {is=struct RigVMUnitNode}.Pose) {variable=dyn_Hair_Back_pose}
  ]}
]
hierarchy [
 root {loc=<0 0 0>, rot=<0 0 0 1>}
 pelvis {parent=root, loc=<0 -0.7321 92.4218>, rot=<0.0127 0.707 -0.0127 -0.707>}
 spine_01 {parent=pelvis, loc=<5.6002 -0.5128 0>, rot=<0 0 0.034 -0.9994>}
 spine_02 {parent=spine_01, loc=<8.0753 0.0139 0>, rot=<0 0 0.0238 -0.9997>}
 spine_03 {parent=spine_02, loc=<7.0393 0 0>, rot=<0 0 0.0149 -0.9999>}
 spine_04 {parent=spine_03, loc=<8.1024 0 0>, rot=<0 0 -0.0132 -0.9999>}
 spine_05 {parent=spine_04, loc=<10.0146 0 0>, rot=<0 0 -0.0015 -1>}
 neck_01 {parent=spine_05, loc=<10.815 1.9196 0>, rot=<0 0 -0.1864 -0.9825>}
 neck_02 {parent=neck_01, loc=<4.7347 0.3045 0>, rot=<0 0 0.0363 -0.9993>}
 head {parent=neck_02, loc=<5.7491 0.5546 0.0012>, rot=<0 0 0.115 -0.9934>}
 dyn_Hair_Back_01 {parent=head, loc=<11.9716 -6.0659 -0.3271>, rot=<0 0 0.0928 -0.9957>}
 dyn_Hair_Back_02 {parent=dyn_Hair_Back_01, loc=<-1.9753 -3.7984 0>, rot=<0 0 0.5593 -0.8289>}
 dyn_Hair_Back_03 {parent=dyn_Hair_Back_02, loc=<-0.1058 -6.9442 0>, rot=<0 0 -0.004 -1>}
 dyn_Hair_Back_04 {parent=dyn_Hair_Back_03, loc=<5.3351 -9.4753 -0.3554>, rot=<0 0 0 1>}
 dyn_Hair_Back_05 {parent=dyn_Hair_Back_04, loc=<2.6338 -9.663 0>, rot=<0 0 0 1>}
 dyn_Hair_Back_06 {parent=dyn_Hair_Back_05, loc=<-0.9292 -7.1818 0>, rot=<0 0 0.0617 -0.9981>}
 dyn_Hair_BackL_01 {parent=head, loc=<11.4422 -5.836 -7.3148>, rot=<0 0 0.0928 -0.9957>}
 dyn_Hair_BackL_02 {parent=dyn_Hair_BackL_01, loc=<-1.9753 -3.7984 0>, rot=<0 0 0.5593 -0.8289>}
 dyn_Hair_BackL_03 {parent=dyn_Hair_BackL_02, loc=<-0.1058 -6.9442 0>, rot=<0 0 -0.004 -1>}
 dyn_Hair_BackL_04 {parent=dyn_Hair_BackL_03, loc=<4.8505 -9.5975 -1.2658>, rot=<0 0 0 1>}
 dyn_Hair_BackL_05 {parent=dyn_Hair_BackL_04, loc=<1.8626 -7.0493 0>, rot=<0 0 0 1>}
 dyn_hair_fringe_l_09 {parent=dyn_Hair_BackL_03, loc=<1.4074 -8.2805 -5.847>, rot=<0.0678 -0.0736 -0.005 -0.995>}
 dyn_Hair_BackR_01 {parent=head, loc=<11.6554 -5.2397 7.0909>, rot=<0 0 0.0928 -0.9957>}
 dyn_Hair_BackR_02 {parent=dyn_Hair_BackR_01, loc=<-1.9753 -3.7984 0>, rot=<0 0 0.5593 -0.8289>}
 dyn_Hair_BackR_03 {parent=dyn_Hair_BackR_02, loc=<-0.1058 -6.9442 0>, rot=<0 0 -0.004 -1>}
 dyn_Hair_BackR_04 {parent=dyn_Hair_BackR_03, loc=<5.6688 -9.5601 -0.6724>, rot=<0 0 0 1>}
 dyn_Hair_BackR_05 {parent=dyn_Hair_BackR_04, loc=<1.8626 -7.0493 0>, rot=<0 0 0 1>}
 dyn_hair_add_r_01 {parent=dyn_Hair_BackR_02, loc=<-2.6064 0.8396 1.3954>, rot=<-0.4831 0.5163 0.4632 -0.5343>}
 dyn_hair_add_r_MSRT_02 {parent=dyn_hair_add_r_01, loc=<10.8949 0 0>, rot=<0 0 0 1>}
 dyn_hair_add_r_02 {parent=dyn_hair_add_r_MSRT_02, loc=<0 0 0>, rot=<0 0 0 1>}
 dyn_hair_fringe_mid_03 {parent=head, loc=<17.774 6.4994 5.064>, rot=<0.8985 -0.314 -0.0925 -0.2925>}
 dyn_hair_fringe_mid_04 {parent=dyn_hair_fringe_mid_03, loc=<8.0701 1.9739 -0.2801>, rot=<0.0581 0.0212 -0.3414 -0.9379>}
 dyn_hair_fringe_mid_05 {parent=dyn_hair_fringe_mid_04, loc=<6.954 2.6891 -0.6218>, rot=<0 0 -0.3321 -0.9432>}
 dyn_hair_fringe_mid_01 {parent=head, loc=<11.0875 8.7202 2.5587>, rot=<-0.1934 0.1633 -0.7774 -0.5758>}
 dyn_hair_fringe_mid_02 {parent=dyn_hair_fringe_mid_01, loc=<2.7418 0 0>, rot=<0 0 0 1>}
 dyn_hair_fringe_r_02 {parent=head, loc=<-3.4844 -3.3112 11.0247>, rot=<0.2042 0.677 0.6875 -0.1654>}
 dyn_hair_fringe_r_03 {parent=dyn_hair_fringe_r_02, loc=<3.7249 0 0>, rot=<0 0 0 1>}
 dyn_hair_fringe_l_03 {parent=head, loc=<12.6233 -6.2344 -11.5708>, rot=<0.5491 -0.7781 -0.1759 -0.2493>}
 dyn_hair_fringe_l_04 {parent=dyn_hair_fringe_l_03, loc=<2.8427 0 0>, rot=<0 0 0 1>}
 dyn_hair_fringe_l_01 {parent=head, loc=<16.7224 -6.3977 -8.7764>, rot=<0.6333 -0.6666 -0.1975 -0.3399>}
 dyn_hair_fringe_l_02 {parent=dyn_hair_fringe_l_01, loc=<2.483 0 0>, rot=<0 0 0 1>}
 dyn_hair_fringe_l_05 {parent=head, loc=<10.7084 -0.9775 -10.6751>, rot=<-0.09 -0.028 0.9708 0.2208>}
 dyn_hair_fringe_l_06 {parent=dyn_hair_fringe_l_05, loc=<4.8903 0 0>, rot=<0 0 0 1>}
 dyn_hair_fringe_r_01 {parent=head, loc=<4.9313 0.1839 12.513>, rot=<-0.2033 0.4769 -0.4912 -0.7>}
 dyn_hair_fringe_l_07 {parent=head, loc=<10.0676 -5.3787 -12.1876>, rot=<0.2765 -0.919 -0.081 -0.2693>}
 dyn_hair_fringe_l_08 {parent=dyn_hair_fringe_l_07, loc=<4.4686 0 0>, rot=<0 -0.2804 0 -0.9599>}
 dyn_Hair_BackL_01_2 {parent=head, loc=<11.4422 -5.836 -7.3148>, rot=<0 0 0.0928 -0.9957>}
 dyn_Hair_BackL_01_3 {parent=head, loc=<11.4422 -5.836 -7.3148>, rot=<0 0 0.0928 -0.9957>}
 clavicle_l {parent=spine_05, loc=<2.5856 -0.2866 -4.3777>, rot=<0.0465 -0.7289 -0.009 -0.6829>}
 upperarm_l {parent=clavicle_l, loc=<18.2264 0.1342 -0.001>, rot=<0.0253 -0.2693 0.0013 -0.9627>}
 clavicle_r {parent=spine_05, loc=<2.5855 -0.2866 4.3777>, rot=<0.7289 0.0465 -0.6829 0.009>}
 upperarm_r {parent=clavicle_r, loc=<-18.2264 -0.1342 0.001>, rot=<0.0253 -0.2693 0.0013 -0.9627>}
 dyn_Hair_Back_01_Null {parent=pelvis, type=Null, loc=<71.3134 -7.0397 -0.3258>, rot=<0 0 0.1148 -0.9934>}
 dyn_Hair_Back_01_ctrl {controlType=EULER_TRANSFORM, parent=dyn_Hair_Back_01_Null, type=Control, loc=<0 0 0>, rot=<0 0 0 1>}
 dyn_Hair_Back_02_Null {parent=pelvis, type=Null, loc=<68.5241 -10.2876 -0.3258>, rot=<0 0 -0.6508 0.7593>}
 dyn_Hair_Back_02_ctrl {controlType=EULER_TRANSFORM, parent=dyn_Hair_Back_02_Null, type=Control, loc=<0 0 0>, rot=<0 0 0 1>}
 dyn_Hair_Back_03_Null {parent=pelvis, type=Null, loc=<61.6455 -11.2455 -0.3258>, rot=<0 0 0.6478 -0.7618>}
 dyn_Hair_Back_03_ctrl {controlType=EULER_TRANSFORM, parent=dyn_Hair_Back_03_Null, type=Control, loc=<0 0 0>, rot=<0 0 0 1>}
 dyn_Hair_Back_04_Null {parent=pelvis, type=Null, loc=<53.1515 -18.035 -0.6812>, rot=<0 0 -0.6478 0.7618>}
 dyn_Hair_Back_04_ctrl {controlType=EULER_TRANSFORM, parent=dyn_Hair_Back_04_Null, type=Control, loc=<0 0 0>, rot=<0 0 0 1>}
 dyn_Hair_Back_05_Null {parent=pelvis, type=Null, loc=<44.0377 -22.1885 -0.6812>, rot=<0 0 0.6478 -0.7618>}
 dyn_Hair_Back_05_ctrl {controlType=EULER_TRANSFORM, parent=dyn_Hair_Back_05_Null, type=Control, loc=<0 0 0>, rot=<0 0 0 1>}
 dyn_Hair_Back_06_Null {parent=pelvis, type=Null, loc=<36.8 -22.4264 -0.6812>, rot=<0 0 -0.6935 0.7204>}
 dyn_Hair_Back_06_ctrl {controlType=EULER_TRANSFORM, parent=dyn_Hair_Back_06_Null, type=Control, loc=<0 0 0>, rot=<0 0 0 1>}
]
variables [
 dyn_Hair_BackL_ctrls {
  type="TArray<FRigElementKey>"
  default=((Type=Control,Name="dyn_Hair_BackL_02"),(Type=Control,Name="dyn_Hair_BackL_03"),(Type=Control,Name="dyn_Hair_BackL_04"),(Type=Control,Name="dyn_Hair_BackL_05"),(Type=Control,Name="dyn_Hair_BackL_06"))
 }
 dyn_Hair_BackL_pose {type=FRigPose, default=(Elements=,HierarchyTopologyVersion=-1,PoseHash=-1)}
 dyn_Hair_BackR_ctrls {
  type="TArray<FRigElementKey>"
  default=((Type=Control,Name="dfrm_arm_spline_01_r_ctrl"),(Type=Control,Name="dfrm_arm_spline_03_r_ctrl"),(Type=Control,Name="dfrm_arm_spline_05_r_ctrl"),(Type=Control,Name="dfrm_arm_spline_07_r_ctrl"),(Type=Control,Name="dfrm_arm_spline_09_r_ctrl"))
 }
 dyn_Hair_BackR_pose {type=FRigPose, default=(Elements=,HierarchyTopologyVersion=-1,PoseHash=-1)}
 dyn_Hair_Back_ctrls {
  type="TArray<FRigElementKey>"
  default=((Type=Control,Name="dyn_Hair_BackL_02"),(Type=Control,Name="dyn_Hair_BackL_03"),(Type=Control,Name="dyn_Hair_BackL_04"),(Type=Control,Name="dyn_Hair_BackL_05"),(Type=Control,Name="dyn_Hair_BackL_06"))
 }
 dyn_Hair_Back_pose {type=FRigPose, default=(Elements=,HierarchyTopologyVersion=-1,PoseHash=-1)}
]
