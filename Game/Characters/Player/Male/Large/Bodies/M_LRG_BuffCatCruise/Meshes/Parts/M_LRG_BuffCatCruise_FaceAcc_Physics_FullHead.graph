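# Physics asset for the full-head face accessory of the large male BuffCatCruise player body, defining dynamic hair bodies and constraints.
# Left, right and middle capsule chains hang from MSTR sphere bodies, while three back chains attach directly to the head.
# Each link has limited swing and tight twist, with wider primary swing on the back chains.
# Head, neck, clavicle and upper-spine shapes have no constraints among themselves, apparently serving as colliders.
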PhysicsAsset /Game/Characters/Player/Male/Large/Bodies/M_LRG_BuffCatCruise/Meshes/Parts/M_LRG_BuffCatCruise_FaceAcc_Physics_FullHead
bodies {
 clavicle_l [{
   shapeType=Capsule
   length=18
   radius=8.3
   shapeName=clavicle_r_capsule
   center=<8.6589 -2.9521 3.6644>
   rotation=<rot -83.1283 179.9988 -179.9988>
  }]
 clavicle_r [{
   shapeType=Capsule
   length=18
   radius=8.7
   shapeName=clavicle_r_capsule
   center=<-8.6589 2.9521 -3.6644>
   rotation=<rot -83.1283 179.9988 0.0012>
  }]
 dyn_Hair_BackL_03 [{shapeType=Capsule, length=2, radius=6.5, shapeName=dyn_Hair_BackL_03_capsule, center=<2.1065 0.7667 0>, rotation=<rot 0 19.9998 90>}]
 dyn_Hair_BackL_04 [{shapeType=Capsule, length=2, radius=6.5, shapeName=dyn_Hair_BackL_04_capsule, center=<2.1065 0.7667 0>, rotation=<rot 0 19.9998 90>}]
 dyn_Hair_BackL_05 [{shapeType=Capsule, length=2, radius=6.5, shapeName=dyn_Hair_BackL_05_capsule, center=<2.1065 0.7667 0>, rotation=<rot 0 19.9998 90>}]
 dyn_Hair_BackR_03 [{shapeType=Capsule, length=2, radius=6.5, shapeName=dyn_Hair_BackR_03_capsule, center=<2.1065 0.7667 0>, rotation=<rot 0 19.9998 90>}]
 dyn_Hair_BackR_04 [{shapeType=Capsule, length=2, radius=6.5, shapeName=dyn_Hair_BackR_04_capsule, center=<2.1065 0.7667 0>, rotation=<rot 0 19.9998 90>}]
 dyn_Hair_BackR_05 [{shapeType=Capsule, length=2, radius=6.5, shapeName=dyn_Hair_BackR_05_capsule, center=<2.1065 0.7667 0>, rotation=<rot 0 19.9998 90>}]
 dyn_Hair_Back_03 [{shapeType=Capsule, length=2, radius=6.5, shapeName=dyn_Hair_Back_03_capsule, center=<2.1065 0.7667 0>, rotation=<rot 0 19.9998 90>}]
 dyn_Hair_Back_04 [{shapeType=Capsule, length=2, radius=6.5, shapeName=dyn_Hair_Back_04_capsule, center=<2.1065 0.7667 0>, rotation=<rot 0 19.9998 90>}]
 dyn_Hair_Back_05 [{shapeType=Capsule, length=2, radius=6.5, shapeName=dyn_Hair_Back_05_capsule, center=<2.1065 0.7667 0>, rotation=<rot 0 19.9998 90>}]
 dyn_Hair_Back_06 [{shapeType=Capsule, length=2, radius=6.5, shapeName=dyn_Hair_Back_06_capsule, center=<2.1065 0.7667 0>, rotation=<rot 0 19.9998 90>}]
 dyn_hair_l_01 [{shapeType=Capsule, length=5, radius=2, shapeName=dyn_hair_l_01_capsule, center=<5 0 0>}]
 dyn_hair_l_02 [{shapeType=Capsule, length=5, radius=2.6, shapeName=dyn_hair_l_02_capsule, center=<5 0 0>}]
 dyn_hair_l_03 [{shapeType=Capsule, length=8, radius=4.5, shapeName=dyn_hair_l_03_capsule, center=<10 -0.3 0>}]
 dyn_hair_l_MSTR [{shapeType=Sphere, radius=2, shapeName=dyn_hair_l_MSTR_sphere}]
 dyn_hair_middle_01 [{shapeType=Capsule, length=5, radius=5.3, shapeName=dyn_hair_middle_01_capsule, center=<5 0 0>}]
 dyn_hair_middle_02 [{shapeType=Capsule, length=5, radius=4.8, shapeName=dyn_hair_middle_02_capsule, center=<5 0 0>}]
 dyn_hair_middle_03 [{shapeType=Capsule, length=5, radius=3.7, shapeName=dyn_hair_middle_03_capsule, center=<5 0 0>}]
 dyn_hair_middle_04 [{shapeType=Capsule, length=5, radius=3.2, shapeName=dyn_hair_middle_04_capsule, center=<5 0 0>, rotation=<rot 0 0.0003 -4.6258>}]
 dyn_hair_middle_05 [{shapeType=Capsule, length=8, radius=5, shapeName=dyn_hair_middle_05_capsule, center=<8.3 -1.3 -4.7>, rotation=<rot 90 -159.5944 -167.5709>}]
 dyn_hair_middle_MSTR [{shapeType=Sphere, radius=2, shapeName=dyn_hair_middle_MSTR_sphere}]
 dyn_hair_middle_MSTR_2 [{shapeType=Sphere, radius=2.7325, shapeName=dyn_hair_middle_MSTR_2_sphere}]
 dyn_hair_r_01 [{shapeType=Capsule, length=5, radius=2, shapeName=dyn_hair_r_01_capsule, center=<5 0 0>}]
 dyn_hair_r_02 [{shapeType=Capsule, length=4.5, radius=3, shapeName=dyn_hair_r_02_capsule, center=<5 0 0>}]
 dyn_hair_r_03 [{shapeType=Capsule, length=8, radius=4.8, shapeName=dyn_hair_r_03_capsule, center=<9.3 0 0>, rotation=<rot -11 0.0003 -7.8484>}]
 dyn_hair_r_MSTR [{shapeType=Sphere, radius=2, shapeName=dyn_hair_r_MSTR_sphere}]
 head [{shapeType=Capsule, length=8, radius=9.8951, shapeName=head_capsule, center=<0 3.1032 0>}]
 neck_01 [{shapeType=Sphere, radius=1, shapeName=neck_01_sphere}]
 neck_02 [{shapeType=Sphere, radius=1, shapeName=neck_02_sphere}]
 spine_04 [{shapeType=Capsule, length=15, radius=21, shapeName=spine_04_capsule, center=<3.6134 3.8892 0>}]
 spine_05 [{shapeType=Capsule, length=11.0116, radius=35, shapeName=spine_05_capsule, center=<-9.4745 16.9639 0>}]
}
constraints [
 {bone1Name=dyn_hair_l_01, bone2Name=dyn_hair_l_MSTR, swing1LimitDegrees=30, swing2LimitDegrees=10, twistLimitDegrees=5}
 {bone1Name=dyn_hair_l_02, bone2Name=dyn_hair_l_01, swing1LimitDegrees=30, swing2LimitDegrees=10, twistLimitDegrees=5}
 {bone1Name=dyn_hair_l_03, bone2Name=dyn_hair_l_02, swing1LimitDegrees=30, swing2LimitDegrees=10, twistLimitDegrees=5}
 {bone1Name=dyn_hair_r_01, bone2Name=dyn_hair_r_MSTR, swing1LimitDegrees=30, swing2LimitDegrees=10, twistLimitDegrees=5}
 {bone1Name=dyn_hair_r_02, bone2Name=dyn_hair_r_01, swing1LimitDegrees=30, swing2LimitDegrees=10, twistLimitDegrees=5}
 {bone1Name=dyn_hair_r_03, bone2Name=dyn_hair_r_02, swing1LimitDegrees=30, swing2LimitDegrees=10, twistLimitDegrees=5}
 {bone1Name=dyn_hair_middle_01, bone2Name=dyn_hair_middle_MSTR_2, swing1LimitDegrees=30, swing2LimitDegrees=10, twistLimitDegrees=5}
 {bone1Name=dyn_hair_middle_02, bone2Name=dyn_hair_middle_01, swing1LimitDegrees=30, swing2LimitDegrees=10, twistLimitDegrees=5}
 {bone1Name=dyn_hair_middle_03, bone2Name=dyn_hair_middle_02, swing1LimitDegrees=30, swing2LimitDegrees=10, twistLimitDegrees=5}
 {bone1Name=dyn_hair_middle_04, bone2Name=dyn_hair_middle_03, swing1LimitDegrees=30, swing2LimitDegrees=5, twistLimitDegrees=5}
 {bone1Name=dyn_hair_middle_05, bone2Name=dyn_hair_middle_04, swing1LimitDegrees=30, swing2LimitDegrees=5, twistLimitDegrees=5}
 {bone1Name=dyn_Hair_BackL_05, bone2Name=dyn_Hair_BackL_04, swing1LimitDegrees=45, swing2LimitDegrees=5, twistLimitDegrees=5}
 {bone1Name=dyn_Hair_BackL_04, bone2Name=dyn_Hair_BackL_03, swing1LimitDegrees=45, swing2LimitDegrees=5, twistLimitDegrees=5}
 {bone1Name=dyn_Hair_BackL_03, bone2Name=head, swing1LimitDegrees=45, swing2LimitDegrees=5, twistLimitDegrees=5}
 {bone1Name=dyn_Hair_BackR_05, bone2Name=dyn_Hair_BackR_04, swing1LimitDegrees=45, swing2LimitDegrees=5, twistLimitDegrees=5}
 {bone1Name=dyn_Hair_BackR_04, bone2Name=dyn_Hair_BackR_03, swing1LimitDegrees=45, swing2LimitDegrees=5, twistLimitDegrees=5}
 {bone1Name=dyn_Hair_BackR_03, bone2Name=head, swing1LimitDegrees=45, swing2LimitDegrees=5, twistLimitDegrees=5}
 {bone1Name=dyn_Hair_Back_06, bone2Name=dyn_Hair_Back_05, swing1LimitDegrees=45, swing2LimitDegrees=5, twistLimitDegrees=5}
 {bone1Name=dyn_Hair_Back_04, bone2Name=dyn_Hair_Back_03, swing1LimitDegrees=45, swing2LimitDegrees=5, twistLimitDegrees=5}
 {bone1Name=dyn_Hair_Back_05, bone2Name=dyn_Hair_Back_04, swing1LimitDegrees=45, swing2LimitDegrees=5, twistLimitDegrees=5}
 {bone1Name=dyn_Hair_Back_03, bone2Name=head, swing1LimitDegrees=45, swing2LimitDegrees=5, twistLimitDegrees=5}
]
