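# Post-process style AnimBP for the M_LRG_BuffCatCruise face accessory, animating the character's hair with procedural secondary motion.
# It copies the parent mesh pose, runs a hair Control Rig, then offsets and simulates back-hair bones with per-bone AnimDynamics and spherical limits.
# Cached poses blend simulated and rigged hair by backpack, skydive and DBNO state.
# A shoulder constraint and spring-driven fringe dynamics follow before output.
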AnimBlueprint /BRCosmetics/Characters/Player/Male/Large/Bodies/M_LRG_BuffCatCruise/Meshes/Parts/M_LRG_BuffCatCruise_FaceAcc_AnimBP
target_skeleton=&/M_LRG_BuffCatCruise_FaceAcc_Skeleton
graphs [
 AnimGraph {nodes [
   n1 = Get is crouching float() -> (is crouching float:double)
   n2 = "Constraint - dyn_hair_middle_04 (P)"(Alpha=@n1, ConstraintWeights_0=1) {
    alphaBoolBlend {blendOption=Linear}
    boneToModify {boneName=dyn_hair_middle_04}
    constraintSetup [{targetBone {boneName=spine_04}}]
    constraintWeights [1]
   }
   n3 = "Constraint - dyn_hair_l_02 (PP)"!(ComponentPose=@n2, Alpha=@n1, ConstraintWeights_0=0.7, ConstraintWeights_1:float=0.3) -> (Pose) {
    alphaBoolBlend {blendOption=Linear}
    alphaScaleBiasClamp {bInterpResult=true}
    boneToModify {boneName=dyn_hair_l_02}
    constraintSetup [{targetBone {boneName=dyn_hair_middle_04}}, {targetBone {boneName=clavicle_l}}]
    constraintWeights [1,1]
   }
   "Constraint - dyn_hair_r_03 (PP)"!(ComponentPose=@n3, Alpha=@n1, ConstraintWeights_0=0.5, ConstraintWeights_1:float=0.5) -> (Pose) {
    alphaBoolBlend {blendOption=Linear}
    alphaScaleBiasClamp {bInterpResult=true}
    boneToModify {boneName=dyn_hair_r_03}
    constraintSetup [{targetBone {boneName=dyn_hair_middle_04}}, {targetBone {boneName=clavicle_r}}]
    constraintWeights [1,1]
   }
   n4 = "Transform (Modify) Bone - Bone: dyn_hair_l_02"!(ComponentPose, bAlphaBoolEnabled:bool=Get bIsCrouchSprinting(), Translation="0,-2,0", Rotation="0,0,0", Scale="1,1,1") -> (Pose) {
    alphaInputType=Bool
    rotationMode=BMM_Additive
    rotationSpace=BCS_BoneSpace
    translationMode=BMM_Additive
    translationSpace=BCS_BoneSpace
    alphaBoolBlend {blendInTime=0.2, blendOption=Sinusoidal, blendOutTime=0.2}
    boneToModify {boneName=dyn_hair_l_02}
    translation {y=-2}
   }
   "Constraint - dyn_hair_l_03 (PP)"!(ComponentPose=@n4, Alpha=Get mele and tactical() -> (mele and tactical:double), ConstraintWeights_0=1, ConstraintWeights_1:float=0) -> (Pose) {
    alphaBoolBlend {blendOption=Linear}
    alphaScaleBiasClamp {bInterpResult=true, interpSpeedDecreasing=1}
    boneToModify {boneName=dyn_hair_l_03}
    constraintSetup [{targetBone {boneName=dyn_hair_middle_04}}, {targetBone {boneName=clavicle_l}}]
    constraintWeights [1,1]
   }
   n5 = Get look at alpha() -> (look at alpha:double)
   n6 = Look At - Bone: dyn_hair_l_MSTR(Alpha=@n5) {
    alphaBoolBlend {blendOption=Linear}
    alphaScaleBiasClamp {bInterpResult=true, interpSpeedDecreasing=0.5}
    boneToModify {boneName=dyn_hair_l_MSTR}
    lookAtLocation=<10 -20 -10>
    lookAtTarget {boneReference {boneName=spine_05}}
    lookAt_Axis {axis=<1 0>}
   }
   n7 = Look At - Bone: dyn_hair_middle_MSTR_2(ComponentPose=@n6, Alpha=@n5) {
    alphaBoolBlend {blendOption=Linear}
    alphaScaleBiasClamp {bInterpResult=true, interpSpeedDecreasing=0.5}
    boneToModify {boneName=dyn_hair_middle_MSTR_2}
    lookAtLocation=<0 -30>
    lookAtTarget {boneReference {boneName=spine_05}}
    lookAt_Axis {axis=<1 0>}
   }
   Look At - Bone: dyn_hair_r_MSTR(ComponentPose=@n7, Alpha=@n5) {
    alphaBoolBlend {blendOption=Linear}
    alphaScaleBiasClamp {bInterpResult=true, interpSpeedDecreasing=0.5}
    boneToModify {boneName=dyn_hair_r_MSTR}
    lookAtLocation=<20 -30 5>
    lookAtTarget {boneReference {boneName=spine_05}}
    lookAt_Axis {axis=<1 0>}
   }
   n8 = "Transform (Modify) Bone - Bone: dyn_hair_middle_MSTR"!(ComponentPose, bAlphaBoolEnabled:bool=Get bIsCrouchSprinting(), Translation="0,-5,0", Rotation="0,-10,0", Scale="1,1,1") -> (Pose) {
    alphaInputType=Bool
    rotationMode=BMM_Additive
    rotationSpace=BCS_BoneSpace
    translationMode=BMM_Additive
    translationSpace=BCS_BoneSpace
    alphaBoolBlend {blendInTime=0.2, blendOption=Sinusoidal, blendOutTime=0.2}
    boneToModify {boneName=dyn_hair_middle_MSTR}
    rotation {yaw=-10}
    translation {y=-5}
   }
   n9 = "Transform (Modify) Bone - Bone: dyn_hair_l_MSTR"!(ComponentPose=@n8, bAlphaBoolEnabled:bool=Get bIsSkyDiving(), Translation="0,-9,0", Rotation="0,0,0", Scale="1,1,1") -> (Pose) {
    alphaInputType=Bool
    rotationMode=BMM_Additive
    rotationSpace=BCS_BoneSpace
    translationMode=BMM_Additive
    translationSpace=BCS_BoneSpace
    alphaBoolBlend {blendInTime=0.2, blendOption=Linear, blendOutTime=0.2}
    boneToModify {boneName=dyn_hair_l_MSTR}
    translation {y=-9}
   }
   n10 = Get bIsPlayingMeleeAnim()
   n11 = "Transform (Modify) Bone - Bone: dyn_hair_l_MSTR"(ComponentPose=@n9, bAlphaBoolEnabled=@n10, Translation="0,0,0", Rotation="0,-10,0", Scale="1,1,1") {
    alphaInputType=Bool
    rotationMode=BMM_Additive
    rotationSpace=BCS_BoneSpace
    translationMode=BMM_Additive
    translationSpace=BCS_BoneSpace
    alphaBoolBlend {blendInTime=0.2, blendOption=Sinusoidal, blendOutTime=0.2}
    boneToModify {boneName=dyn_hair_l_MSTR}
    rotation {yaw=-10}
   }
   n12 = "Transform (Modify) Bone - Bone: dyn_hair_r_MSTR"!(ComponentPose=@n11, bAlphaBoolEnabled:bool=@n10, Translation="0,0,0", Rotation="0,-10,0", Scale="1,1,1") -> (Pose) {
    alphaInputType=Bool
    rotationMode=BMM_Additive
    rotationSpace=BCS_BoneSpace
    translationMode=BMM_Additive
    translationSpace=BCS_BoneSpace
    alphaBoolBlend {blendInTime=0.2, blendOption=Sinusoidal, blendOutTime=0.2}
    boneToModify {boneName=dyn_hair_r_MSTR}
    rotation {yaw=-10}
   }
   n13 = Get head up or to the shoulder() -> (head up or to the shoulder:double)
   n14 = "Transform (Modify) Bone - Bone: dyn_hair_middle_MSTR"!(ComponentPose=@n12, Alpha:float=@n13, Translation="0,0,0", Rotation="15,0,0", Scale="1,1,1") -> (Pose) {
    rotationMode=BMM_Additive
    rotationSpace=BCS_BoneSpace
    scaleMode=BMM_Replace
    scaleSpace=BCS_BoneSpace
    translationMode=BMM_Additive
    translationSpace=BCS_BoneSpace
    alphaBoolBlend {blendOption=Linear}
    boneToModify {boneName=dyn_hair_middle_MSTR}
    rotation {pitch=15}
   }
   n15 = "Transform (Modify) Bone - Bone: dyn_hair_l_MSTR"!(ComponentPose=@n14, Alpha:float=@n13, Translation="0,0,0", Rotation="0,0,0", Scale="1,1,1") -> (Pose) {
    rotationMode=BMM_Additive
    rotationSpace=BCS_BoneSpace
    scaleMode=BMM_Replace
    scaleSpace=BCS_BoneSpace
    translationMode=BMM_Additive
    translationSpace=BCS_BoneSpace
    alphaBoolBlend {blendOption=Linear}
    boneToModify {boneName=dyn_hair_l_MSTR}
   }
   n16 = "Transform (Modify) Bone - Bone: dyn_hair_r_MSTR"!(ComponentPose=@n15, Alpha:float=@n13, Translation="-8,-12,0", Rotation="15,0,0", Scale="1,1,1") -> (Pose) {
    rotationMode=BMM_Additive
    rotationSpace=BCS_BoneSpace
    scaleMode=BMM_Replace
    scaleSpace=BCS_BoneSpace
    translationMode=BMM_Additive
    translationSpace=BCS_BoneSpace
    alphaBoolBlend {blendOption=Linear}
    boneToModify {boneName=dyn_hair_r_MSTR}
    rotation {pitch=15}
    translation=<-8 -12>
   }
   "Transform (Modify) Bone - Bone: dyn_hair_add_r_02"(ComponentPose=@n16, Alpha=@n13, Translation="0,0,-5", Rotation="0,0,0", Scale="1,1,1") {
    rotationMode=BMM_Additive
    rotationSpace=BCS_BoneSpace
    translationMode=BMM_Additive
    translationSpace=BCS_BoneSpace
    alphaBoolBlend {blendOption=Linear}
    boneToModify {boneName=dyn_hair_add_r_02}
    translation {z=-5}
   }
   RigidBody!(ComponentPose, Alpha, OverrideWorldGravity:Vector=Get Gravity Override(), ExternalForce) -> (Pose) {
    bClampLinearTranslationLimitToRefPose=true
    bOverrideWorldGravity=true
    overridePhysicsAsset=/Game/Characters/Player/Male/Large/Bodies/M_LRG_BuffCatCruise/Meshes/Parts/M_LRG_BuffCatCruise_FaceAcc_Physics_FullHead
    simulationSpace=BaseBoneSpace
    alphaBoolBlend {blendOption=Linear}
    baseBoneRef {boneName=neck_02}
    componentAppliedLinearAccClamp=<3000 3000 3000>
    componentLinearAccScale=<0.2 0.4 0.3>
    componentLinearVelScale=<0.3 0.4 0.3>
    overrideWorldGravity {z=-1111}
   }
   n17 = "Anim Dynamics - Chain: dyn_hair_fringe_mid_01 -> dyn_hair_fringe_mid_02"(ComponentPose=Local To Component(LocalPose=Use cached pose 'mainHair'()), Alpha=1, ExternalForce="0,0,0") {
    angularDampingOverride=0.999
    angularSpringConstant=100
    bAngularSpring=true
    bChain=true
    bGravityOverrideInSimSpace=true
    bOverrideAngularDamping=true
    simulationSpace=RootRelative
    alphaBoolBlend {blendOption=Linear}
    alphaScaleBiasClamp {bInterpResult=true}
    boundBone {boneName=dyn_hair_fringe_mid_01}
    chainEnd {boneName=dyn_hair_fringe_mid_02}
    componentAppliedLinearAccClamp=<3500 3500 3500>
    componentLinearAccScale=<0.8 0.8 1.2>
    componentLinearVelScale=<0.8 0.8 1.2>
    physicsBodyDefinitions [
     {
      sphereCollisionRadius=0
      boundBone {boneName=dyn_hair_fringe_mid_01}
      boxExtents=<10 10 10>
      constraintSetup {angularLimitsMax=<15 15 0>, angularLimitsMin=<-15 -3 -35>, angularTarget=<1 0 0>}
      localJointOffset=<0 0 0>
     }
     {
      sphereCollisionRadius=0
      boundBone {boneName=dyn_hair_fringe_mid_02}
      boxExtents=<20 20 10>
      constraintSetup {angularLimitsMax=<15 15 0>, angularLimitsMin=<-15 -3 -35>, angularTarget=<1 0 0>}
      localJointOffset=<2 0 0>
     }
    ]
    relativeSpaceBone {boneName=root}
   }
   n18 = Anim Dynamics - Bone: dyn_hair_fringe_l_01(ComponentPose=@n17, Alpha=1, ExternalForce="0,0,0") {
    angularBiasOverride=1
    angularDampingOverride=1
    angularSpringConstant=700
    bAngularSpring=true
    bOverrideAngularDamping=true
    bUseGravityOverride=true
    linearSpringConstant=7500
    simulationSpace=RootRelative
    alphaBoolBlend {blendOption=Linear}
    boundBone {boneName=dyn_hair_fringe_l_01}
    componentAppliedLinearAccClamp=<3800 3800 3000>
    componentLinearAccScale=<0.3 0.3 0.4>
    componentLinearVelScale=<1.2 1.2 0.4>
    physicsBodyDefinitions [{
      boundBone {boneName=dyn_hair_fringe_l_01}
      boxExtents=<7 7 5>
      constraintSetup {angularLimitsMax=<20 10 30>, angularLimitsMin=<-20 -10 -30>, angularTarget=<1 0 0>}
      localJointOffset=<0.6 0 2>
     }]
    relativeSpaceBone {boneName=pelvis}
   }
   n19 = Anim Dynamics - Bone: dyn_hair_fringe_l_02(ComponentPose=@n18, Alpha=1, ExternalForce="0,0,0") {
    angularDampingOverride=1
    angularSpringConstant=500
    bAngularSpring=true
    bOverrideAngularDamping=true
    bUseGravityOverride=true
    linearSpringConstant=7500
    simulationSpace=RootRelative
    alphaBoolBlend {blendOption=Linear}
    boundBone {boneName=dyn_hair_fringe_l_02}
    componentAppliedLinearAccClamp=<3800 3800 3000>
    componentLinearAccScale=<0.5 0.5 0.2>
    componentLinearVelScale=<1.2 1.2 0.2>
    physicsBodyDefinitions [{
      boundBone {boneName=dyn_hair_fringe_l_02}
      boxExtents=<4 5 3>
      constraintSetup {angularLimitsMax=<20 10 5>, angularLimitsMin=<-20 -10 -30>, angularTarget=<1 0 0>}
      localJointOffset=<0.3 0 1>
     }]
    relativeSpaceBone {boneName=pelvis}
   }
   n20 = Anim Dynamics - Bone: dyn_hair_fringe_l_03(ComponentPose=@n19, Alpha=1, ExternalForce="0,0,0") {
    angularBiasOverride=1
    angularDampingOverride=1
    angularSpringConstant=700
    bAngularSpring=true
    bOverrideAngularDamping=true
    bUseGravityOverride=true
    linearSpringConstant=7500
    simulationSpace=RootRelative
    alphaBoolBlend {blendOption=Linear}
    boundBone {boneName=dyn_hair_fringe_l_03}
    componentAppliedLinearAccClamp=<3800 3800 3000>
    componentLinearAccScale=<0.3 0.3 0.35>
    componentLinearVelScale=<1.2 1.2 0.35>
    physicsBodyDefinitions [{
      boundBone {boneName=dyn_hair_fringe_l_03}
      boxExtents=<9 9 7>
      constraintSetup {angularLimitsMax=<20 6 20>, angularLimitsMin=<-20 -6 -10>, angularTarget=<1 0 0>}
      localJointOffset=<0.6 0 1.7>
     }]
    relativeSpaceBone {boneName=pelvis}
   }
   n21 = Anim Dynamics - Bone: dyn_hair_fringe_l_04(ComponentPose=@n20, Alpha=1, ExternalForce="0,0,0") {
    angularDampingOverride=1
    angularSpringConstant=500
    bAngularSpring=true
    bOverrideAngularDamping=true
    bUseGravityOverride=true
    linearSpringConstant=7500
    simulationSpace=RootRelative
    alphaBoolBlend {blendOption=Linear}
    boundBone {boneName=dyn_hair_fringe_l_04}
    componentAppliedLinearAccClamp=<3800 3800 3000>
    componentLinearAccScale=<0.5 0.5 0.2>
    componentLinearVelScale=<1.2 1.2 0.2>
    physicsBodyDefinitions [{
      boundBone {boneName=dyn_hair_fringe_l_04}
      boxExtents=<6 7 4>
      constraintSetup {angularLimitsMax=<20 6 20>, angularLimitsMin=<-20 -6 -10>, angularTarget=<1 0 0>}
      localJointOffset=<0.2 0 0.6>
     }]
    relativeSpaceBone {boneName=pelvis}
   }
   n22 = Anim Dynamics - Bone: dyn_hair_fringe_l_07(ComponentPose=@n21, Alpha=1, ExternalForce="0,0,0") {
    angularBiasOverride=1
    angularDampingOverride=0.999
    angularSpringConstant=700
    bAngularSpring=true
    bOverrideAngularDamping=true
    bUseGravityOverride=true
    linearSpringConstant=7500
    simulationSpace=RootRelative
    alphaBoolBlend {blendOption=Linear}
    boundBone {boneName=dyn_hair_fringe_l_07}
    componentAppliedLinearAccClamp=<3800 300 300>
    componentLinearAccScale=<0.3 0.15 0.2>
    componentLinearVelScale=<1.2 0.4 0.2>
    physicsBodyDefinitions [{
      boundBone {boneName=dyn_hair_fringe_l_07}
      boxExtents=<12 8 15>
      constraintSetup {angularLimitsMax=<20 6 15>, angularLimitsMin=<-20 -12 -15>, angularTarget=<1 0 0>}
      localJointOffset=<0.6 0.7 2>
     }]
    relativeSpaceBone {boneName=pelvis}
   }
   n23 = Anim Dynamics - Bone: dyn_hair_fringe_l_08(ComponentPose=@n22, Alpha=1, ExternalForce="0,0,0") {
    angularDampingOverride=0.999
    angularSpringConstant=500
    bAngularSpring=true
    bOverrideAngularDamping=true
    bUseGravityOverride=true
    linearSpringConstant=7500
    simulationSpace=RootRelative
    alphaBoolBlend {blendOption=Linear}
    boundBone {boneName=dyn_hair_fringe_l_08}
    componentAppliedLinearAccClamp=<3800 300 300>
    componentLinearAccScale=<0.5 0.15 0.2>
    componentLinearVelScale=<1.2 0.2 0.2>
    physicsBodyDefinitions [{
      boundBone {boneName=dyn_hair_fringe_l_08}
      boxExtents=<10 7 12>
      constraintSetup {angularLimitsMax=<20 0 3>, angularLimitsMin=<-20 -7 -8>, angularTarget=<1 0 0>}
      localJointOffset=<0.2 0.4 0.6>
     }]
    relativeSpaceBone {boneName=pelvis}
   }
   n24 = Anim Dynamics - Bone: dyn_hair_fringe_r_01(ComponentPose=@n23, Alpha=1, ExternalForce="0,0,0") {
    angularBiasOverride=1
    angularDampingOverride=1
    angularSpringConstant=700
    bAngularSpring=true
    bOverrideAngularDamping=true
    bUseGravityOverride=true
    linearSpringConstant=7500
    simulationSpace=RootRelative
    alphaBoolBlend {blendOption=Linear}
    boundBone {boneName=dyn_hair_fringe_r_01}
    componentAppliedLinearAccClamp=<3800 3800 3000>
    componentLinearAccScale=<0.3 0.3 0.4>
    componentLinearVelScale=<1.2 1.2 0.5>
    physicsBodyDefinitions [{
      boundBone {boneName=dyn_hair_fringe_r_01}
      boxExtents=<7 6 6>
      constraintSetup {angularLimitsMax=<20 20 30>, angularLimitsMin=<-20 -20 -30>, angularTarget=<1 0 0>}
      localJointOffset=<0.6 0 2>
     }]
    relativeSpaceBone {boneName=pelvis}
   }
   n25 = Anim Dynamics - Bone: dyn_hair_fringe_r_02(ComponentPose=@n24, Alpha=1, ExternalForce="0,0,0") {
    angularBiasOverride=1
    angularDampingOverride=1
    angularSpringConstant=700
    bAngularSpring=true
    bOverrideAngularDamping=true
    bUseGravityOverride=true
    bUseSphericalLimits=true
    linearSpringConstant=7500
    simulationSpace=RootRelative
    alphaBoolBlend {blendOption=Linear}
    boundBone {boneName=dyn_hair_fringe_r_02}
    componentAppliedLinearAccClamp=<3800 3800 3000>
    componentLinearAccScale=<0.3 0.3 0.4>
    componentLinearVelScale=<1.2 1.2 0.4>
    physicsBodyDefinitions [{
      boundBone {boneName=dyn_hair_fringe_r_02}
      boxExtents=<7 7 4>
      constraintSetup {angularLimitsMax=<20 10 45>, angularLimitsMin=<-20 -10 -30>, angularTarget=<1 0 0>}
      localJointOffset=<0.6 0 2>
     }]
    relativeSpaceBone {boneName=pelvis}
    sphericalLimits [
     {limitRadius=0, drivingBone {boneName=None}, sphereLocalOffset=<0 0 0>}
     {limitRadius=13, drivingBone {boneName=neck_01}, sphereLocalOffset=<0 0 0>}
    ]
   }
   n26 = Anim Dynamics - Bone: dyn_hair_fringe_r_03(ComponentPose=@n25, Alpha=1, ExternalForce="0,0,0") {
    angularDampingOverride=1
    angularSpringConstant=500
    bAngularSpring=true
    bOverrideAngularDamping=true
    bUseGravityOverride=true
    linearSpringConstant=7500
    simulationSpace=RootRelative
    alphaBoolBlend {blendOption=Linear}
    boundBone {boneName=dyn_hair_fringe_r_03}
    componentAppliedLinearAccClamp=<3800 3800 3000>
    componentLinearAccScale=<0.5 0.5 0.2>
    componentLinearVelScale=<1.2 1.2 0.2>
    physicsBodyDefinitions [{
      boundBone {boneName=dyn_hair_fringe_r_03}
      boxExtents=<4 5 3>
      constraintSetup {angularLimitsMax=<20 10 5>, angularLimitsMin=<-20 -10 -30>, angularTarget=<1 0 0>}
      localJointOffset=<0.3 0 1>
     }]
    relativeSpaceBone {boneName=pelvis}
   }
   n27 = Anim Dynamics - Bone: dyn_hair_fringe_l_05(ComponentPose=@n26, Alpha=1, ExternalForce="0,0,0") {
    angularBiasOverride=1
    angularDampingOverride=1
    angularSpringConstant=700
    bAngularSpring=true
    bOverrideAngularDamping=true
    bUseGravityOverride=true
    bUseSphericalLimits=true
    linearSpringConstant=7500
    simulationSpace=RootRelative
    alphaBoolBlend {blendOption=Linear}
    boundBone {boneName=dyn_hair_fringe_l_05}
    componentAppliedLinearAccClamp=<3800 3800 3000>
    componentLinearAccScale=<0.3 0.3 0.4>
    componentLinearVelScale=<1.2 1.2 0.4>
    physicsBodyDefinitions [{
      boundBone {boneName=dyn_hair_fringe_l_05}
      boxExtents=<7 7 5>
      constraintSetup {angularLimitsMax=<0 0 15>, angularLimitsMin=<0 -10 -15>, angularTarget=<1 0 0>}
      localJointOffset=<0.6 0 2>
     }]
    relativeSpaceBone {boneName=pelvis}
    sphericalLimits [{limitRadius=12, drivingBone {boneName=head}, sphereLocalOffset=<0 0 0>}]
   }
   n28 = Anim Dynamics - Bone: dyn_hair_fringe_l_06(ComponentPose=@n27, Alpha=1, ExternalForce="0,0,0") {
    angularDampingOverride=1
    angularSpringConstant=500
    bAngularSpring=true
    bOverrideAngularDamping=true
    bUseGravityOverride=true
    bUseSphericalLimits=true
    linearSpringConstant=7500
    simulationSpace=RootRelative
    alphaBoolBlend {blendOption=Linear}
    boundBone {boneName=dyn_hair_fringe_l_06}
    componentAppliedLinearAccClamp=<3800 3800 3000>
    componentLinearAccScale=<0.5 0.5 0.2>
    componentLinearVelScale=<1.2 1.2 0.2>
    physicsBodyDefinitions [{
      boundBone {boneName=dyn_hair_fringe_l_06}
      boxExtents=<4 5 3>
      constraintSetup {angularLimitsMax=<0 0 8>, angularLimitsMin=<0 -10 -8>, angularTarget=<1 0 0>}
      localJointOffset=<0.3 0 1>
     }]
    relativeSpaceBone {boneName=pelvis}
    sphericalLimits [{limitRadius=13.1, drivingBone {boneName=head}, sphereLocalOffset=<0 0 0>}]
   }
   n29 = Anim Dynamics - Bone: dyn_hair_fringe_mid_03(ComponentPose=@n28, Alpha=1, ExternalForce="0,0,0") {
    angularBiasOverride=1
    angularDampingOverride=0.95
    angularSpringConstant=150
    bAngularSpring=true
    bOverrideAngularDamping=true
    bUseGravityOverride=true
    linearSpringConstant=7500
    simulationSpace=BoneRelative
    alphaBoolBlend {blendOption=Linear}
    boundBone {boneName=dyn_hair_fringe_mid_03}
    componentAppliedLinearAccClamp=<350 350 350>
    componentLinearAccScale=<0.1 0.1 0.1>
    componentLinearVelScale=<0.1 0.1 0.1>
    physicsBodyDefinitions [{
      boundBone {boneName=dyn_hair_fringe_mid_03}
      boxExtents=<8 8 8>
      constraintSetup {angularLimitsMax=<3 0 5>, angularLimitsMin=<-3 0 -20>, angularTarget=<1 0 0>}
      localJointOffset=<0.8 0 0>
     }]
    relativeSpaceBone {boneName=neck_02}
   }
   n30 = Anim Dynamics - Bone: dyn_hair_fringe_mid_04(ComponentPose=@n29, Alpha=1, ExternalForce="0,0,0") {
    angularBiasOverride=1
    angularDampingOverride=0.95
    angularSpringConstant=200
    bAngularSpring=true
    bOverrideAngularDamping=true
    bUseGravityOverride=true
    linearSpringConstant=7500
    simulationSpace=BoneRelative
    alphaBoolBlend {blendOption=Linear}
    boundBone {boneName=dyn_hair_fringe_mid_04}
    componentAppliedLinearAccClamp=<350 350 350>
    componentLinearAccScale=<0.1 0.1 0.1>
    componentLinearVelScale=<0.1 0.1 0.1>
    physicsBodyDefinitions [{
      boundBone {boneName=dyn_hair_fringe_mid_04}
      boxExtents=<8 8 8>
      constraintSetup {angularLimitsMax=<5 5 5>, angularLimitsMin=<-5 -5 -20>, angularTarget=<1 0 0>}
      localJointOffset=<0.5 0 0>
     }]
    relativeSpaceBone {boneName=head}
   }
   n31 = Anim Dynamics - Bone: dyn_hair_fringe_mid_05(ComponentPose=@n30, Alpha=1, ExternalForce="0,0,0") {
    angularBiasOverride=1
    angularDampingOverride=0.95
    angularSpringConstant=100
    bAngularSpring=true
    bOverrideAngularDamping=true
    bUseGravityOverride=true
    bUsePlanarLimit=false
    bUseSphericalLimits=true
    linearSpringConstant=7500
    simulationSpace=BoneRelative
    alphaBoolBlend {blendOption=Linear}
    boundBone {boneName=dyn_hair_fringe_mid_05}
    componentAppliedLinearAccClamp=<350 350 350>
    componentLinearAccScale=<0.1 0.1 0.1>
    componentLinearVelScale=<0.1 0.1 0.1>
    physicsBodyDefinitions [{
      boundBone {boneName=dyn_hair_fringe_mid_05}
      boxExtents=<8 8 8>
      constraintSetup {angularLimitsMax=<2 0 8>, angularLimitsMin=<-2 0 -20>, angularTarget=<1 0 0>}
      localJointOffset=<0.5 0 0>
     }]
    relativeSpaceBone {boneName=head}
    sphericalLimits [{limitRadius=21.5, drivingBone {boneName=head}, sphereLocalOffset=<0 0 0>}]
   }
   Output Pose(Result=Component To Local(ComponentPose=@n31)) {layerGroup=None}
   n32 = Control Rig(Source=Copy Pose From Mesh() {bUseAttachedParent=true}) {
    alphaBoolBlend {blendOption=Linear}
    controlRigAssetReference {blueprintRigClass=&/M_LRG_BuffCatCruise_FaceAcc_HairCtrlRig.M_LRG_BuffCatCruise_FaceAcc_HairCtrlRig_C}
   }
   n33 = "Transform (Modify) Bone - Bone: dyn_Hair_BackL_02"(ComponentPose=Local To Component(LocalPose=@n32), Alpha=1, Translation="0,0,0", Rotation="0,5,0", Scale="1,1,1") {
    rotationMode=BMM_Additive
    rotationSpace=BCS_BoneSpace
    scaleMode=BMM_Additive
    scaleSpace=BCS_BoneSpace
    translationMode=BMM_Additive
    translationSpace=BCS_BoneSpace
    alphaBoolBlend {blendOption=Linear}
    boneToModify {boneName=dyn_Hair_BackL_02}
    rotation {yaw=5}
   }
   n34 = "Transform (Modify) Bone - Bone: dyn_Hair_BackL_03"(ComponentPose=@n33, Alpha=1, Translation="0,0,0", Rotation="0,5,0", Scale="1,1,1") {
    rotationMode=BMM_Additive
    rotationSpace=BCS_BoneSpace
    scaleMode=BMM_Additive
    scaleSpace=BCS_BoneSpace
    translationMode=BMM_Additive
    translationSpace=BCS_BoneSpace
    alphaBoolBlend {blendOption=Linear}
    boneToModify {boneName=dyn_Hair_BackL_03}
    rotation {yaw=5}
   }
   n35 = "Transform (Modify) Bone - Bone: dyn_Hair_Back_02"(ComponentPose=@n34, Alpha=1, Translation="0,0,0", Rotation="0,5,0", Scale="1,1,1") {
    rotationMode=BMM_Additive
    rotationSpace=BCS_BoneSpace
    scaleMode=BMM_Additive
    scaleSpace=BCS_BoneSpace
    translationMode=BMM_Additive
    translationSpace=BCS_BoneSpace
    alphaBoolBlend {blendOption=Linear}
    boneToModify {boneName=dyn_Hair_Back_02}
    rotation {yaw=5}
   }
   n36 = "Transform (Modify) Bone - Bone: dyn_Hair_Back_03"(ComponentPose=@n35, Alpha=1, Translation="0,0,0", Rotation="0,5,0", Scale="1,1,1") {
    rotationMode=BMM_Additive
    rotationSpace=BCS_BoneSpace
    scaleMode=BMM_Additive
    scaleSpace=BCS_BoneSpace
    translationMode=BMM_Additive
    translationSpace=BCS_BoneSpace
    alphaBoolBlend {blendOption=Linear}
    boneToModify {boneName=dyn_Hair_Back_03}
    rotation {yaw=5}
   }
   n37 = "Transform (Modify) Bone - Bone: dyn_Hair_BackR_02"(ComponentPose=@n36, Alpha=1, Translation="0,0,0", Rotation="0,5,0", Scale="1,1,1") {
    rotationMode=BMM_Additive
    rotationSpace=BCS_BoneSpace
    scaleMode=BMM_Additive
    scaleSpace=BCS_BoneSpace
    translationMode=BMM_Additive
    translationSpace=BCS_BoneSpace
    alphaBoolBlend {blendOption=Linear}
    boneToModify {boneName=dyn_Hair_BackR_02}
    rotation {yaw=5}
   }
   n38 = "Transform (Modify) Bone - Bone: dyn_Hair_BackR_03"(ComponentPose=@n37, Alpha=1, Translation="0,0,0", Rotation="0,5,0", Scale="1,1,1") {
    rotationMode=BMM_Additive
    rotationSpace=BCS_BoneSpace
    scaleMode=BMM_Additive
    scaleSpace=BCS_BoneSpace
    translationMode=BMM_Additive
    translationSpace=BCS_BoneSpace
    alphaBoolBlend {blendOption=Linear}
    boneToModify {boneName=dyn_Hair_BackR_03}
    rotation {yaw=5}
   }
   Save cached pose 'HairCachePose'(Pose=Component To Local(ComponentPose=@n38)) {cacheName=HairCachePose}
   n39 = "Constraint - dyn_hair_add_r_02 (P)"(ComponentPose=Local To Component(LocalPose=Use cached pose 'Backpack_on/off'()), Alpha=0.8, ConstraintWeights_0=1) {
    alphaBoolBlend {blendOption=Linear}
    boneToModify {boneName=dyn_hair_add_r_02}
    constraintSetup [{targetBone {boneName=clavicle_r}}]
    constraintWeights [1]
   }
   n40 = "Transform (Modify) Bone - Bone: dyn_hair_add_r_01"!(ComponentPose=@n39, bAlphaBoolEnabled:bool=Get bIsCrouching(), Translation="-3.5,0,0", Rotation="0,0,0", Scale="1,1,1") -> (Pose) {
    alphaInputType=Bool
    rotationMode=BMM_Additive
    rotationSpace=BCS_BoneSpace
    translationMode=BMM_Additive
    translationSpace=BCS_BoneSpace
    alphaBoolBlend {blendInTime=0.2, blendOption=Linear, blendOutTime=0.2}
    boneToModify {boneName=dyn_hair_add_r_01}
    translation {x=-3.5}
   }
   Save cached pose 'mainHair'(Pose=Component To Local(ComponentPose=@n40)) {cacheName=mainHair}
   Save cached pose 'preSim'(Pose=Blend Poses by bool(
     BlendPose_0=Use cached pose 'Backpack_on/off'()
     BlendPose_1=Use cached pose 'HairCachePose'()
     bActiveValue=Get bAnimDynamicsEnabled() -> (bAnimDynamicsEnabled:bool)
    ) {blendType=Linear}) {cacheName=preSim}
   n41 = Use cached pose 'HairDynPose'()
   n42 = Use cached pose 'HairCachePose'()
   n43 = SelectFloat(A=0.85, B=0.15, bPickA=OR Boolean(A=Get isWearingPackBool() -> (isWearingPackBool:bool), B=Get bIsDBNO()))
   n44 = Get isWearingPackBool()
   n45 = Get bIsSkydiveDiveMode()
   n46 = Get bIsDBNO()
   n47 = Get bIsSkydiveFloating()
   Save cached pose 'Backpack_on/off'(Pose=Blend Poses by bool(
     BlendPose_0=Two Way Blend(A=@n41, B=@n42, Alpha=@n43) {blendNode {alphaBoolBlend {blendOption=Linear}}}
     BlendPose_1=Use cached pose 'HairDynPose'()
     bActiveValue=OR Boolean!(A=@n45, B=@n47, C:bool=@n44, D:bool=@n46) -> (ReturnValue)
    ) {blendType=Linear}) {cacheName=Backpack_on/off}
   n48 = "Transform (Modify) Bone - Bone: dyn_Hair_BackL_04"(
    ComponentPose=Local To Component(LocalPose=Use cached pose 'HairCachePose'())
    Alpha=1
    Translation="0,0,0"
    Rotation="0,10,0"
    Scale="1,1,1"
   ) {
    rotationMode=BMM_Additive
    rotationSpace=BCS_BoneSpace
    scaleMode=BMM_Additive
    scaleSpace=BCS_BoneSpace
    translationMode=BMM_Additive
    translationSpace=BCS_BoneSpace
    alphaBoolBlend {blendOption=Linear}
    boneToModify {boneName=dyn_Hair_BackL_04}
    rotation {yaw=10}
   }
   n49 = "Transform (Modify) Bone - Bone: dyn_Hair_BackL_05"(ComponentPose=@n48, Alpha=1, Translation="0,0,0", Rotation="0,10,0", Scale="1,1,1") {
    rotationMode=BMM_Additive
    rotationSpace=BCS_BoneSpace
    scaleMode=BMM_Additive
    scaleSpace=BCS_BoneSpace
    translationMode=BMM_Additive
    translationSpace=BCS_BoneSpace
    alphaBoolBlend {blendOption=Linear}
    boneToModify {boneName=dyn_Hair_BackL_05}
    rotation {yaw=10}
   }
   n50 = "Transform (Modify) Bone - Bone: dyn_Hair_Back_04"(ComponentPose=@n49, Alpha=1, Translation="0,0,0", Rotation="0,10,0", Scale="1,1,1") {
    rotationMode=BMM_Additive
    rotationSpace=BCS_BoneSpace
    scaleMode=BMM_Additive
    scaleSpace=BCS_BoneSpace
    translationMode=BMM_Additive
    translationSpace=BCS_BoneSpace
    alphaBoolBlend {blendOption=Linear}
    boneToModify {boneName=dyn_Hair_Back_04}
    rotation {yaw=10}
   }
   n51 = "Transform (Modify) Bone - Bone: dyn_Hair_Back_05"(ComponentPose=@n50, Alpha=1, Translation="0,0,0", Rotation="0,10,0", Scale="1,1,1") {
    rotationMode=BMM_Additive
    rotationSpace=BCS_BoneSpace
    scaleMode=BMM_Additive
    scaleSpace=BCS_BoneSpace
    translationMode=BMM_Additive
    translationSpace=BCS_BoneSpace
    alphaBoolBlend {blendOption=Linear}
    boneToModify {boneName=dyn_Hair_Back_05}
    rotation {yaw=10}
   }
   n52 = "Transform (Modify) Bone - Bone: dyn_Hair_Back_06"(ComponentPose=@n51, Alpha=1, Translation="0,0,0", Rotation="0,10,0", Scale="1,1,1") {
    rotationMode=BMM_Additive
    rotationSpace=BCS_BoneSpace
    scaleMode=BMM_Additive
    scaleSpace=BCS_BoneSpace
    translationMode=BMM_Additive
    translationSpace=BCS_BoneSpace
    alphaBoolBlend {blendOption=Linear}
    boneToModify {boneName=dyn_Hair_Back_06}
    rotation {yaw=10}
   }
   n53 = "Transform (Modify) Bone - Bone: dyn_Hair_BackR_04"(ComponentPose=@n52, Alpha=1, Translation="0,0,0", Rotation="0,10,0", Scale="1,1,1") {
    rotationMode=BMM_Additive
    rotationSpace=BCS_BoneSpace
    scaleMode=BMM_Additive
    scaleSpace=BCS_BoneSpace
    translationMode=BMM_Additive
    translationSpace=BCS_BoneSpace
    alphaBoolBlend {blendOption=Linear}
    boneToModify {boneName=dyn_Hair_BackR_04}
    rotation {yaw=10}
   }
   n54 = "Transform (Modify) Bone - Bone: dyn_Hair_BackR_05"(ComponentPose=@n53, Alpha=1, Translation="0,0,0", Rotation="0,10,0", Scale="1,1,1") {
    rotationMode=BMM_Additive
    rotationSpace=BCS_BoneSpace
    scaleMode=BMM_Additive
    scaleSpace=BCS_BoneSpace
    translationMode=BMM_Additive
    translationSpace=BCS_BoneSpace
    alphaBoolBlend {blendOption=Linear}
    boneToModify {boneName=dyn_Hair_BackR_05}
    rotation {yaw=10}
   }
   n55 = Anim Dynamics - Bone: dyn_Hair_BackL_02(ComponentPose=@n54, Alpha=0.2, ExternalForce="0,0,0") {
    angularDampingOverride=0.99
    angularSpringConstant=75
    bAngularSpring=true
    bGravityOverrideInSimSpace=true
    bOverrideAngularDamping=true
    bUseGravityOverride=true
    simulationSpace=RootRelative
    alphaBoolBlend {blendOption=Linear}
    boundBone {boneName=dyn_Hair_BackL_02}
    componentAppliedLinearAccClamp=<3500 3500 3000>
    componentLinearAccScale=<0.5 0.5 0.5>
    componentLinearVelScale=<0.65 0.65 0.65>
    gravityOverride {z=-980}
    physicsBodyDefinitions [{
      collisionType=CustomSphere
      sphereCollisionRadius=2
      boundBone {boneName=dyn_Hair_BackL_02}
      boxExtents=<40 40 40>
      constraintSetup {angularLimitsMax=<0 0 5>, angularLimitsMin=<0 0 -3>, angularTarget=<0 0 0>}
      localJointOffset=<-2 -5 0>
     }]
    relativeSpaceBone {boneName=root}
   }
   n56 = Anim Dynamics - Bone: dyn_Hair_BackL_03(ComponentPose=@n55, Alpha=1, ExternalForce="0,0,0") {
    angularDampingOverride=0.99
    angularSpringConstant=75
    bAngularSpring=true
    bGravityOverrideInSimSpace=true
    bOverrideAngularDamping=true
    bUseGravityOverride=true
    simulationSpace=RootRelative
    alphaBoolBlend {blendOption=Linear}
    boundBone {boneName=dyn_Hair_BackL_03}
    componentAppliedLinearAccClamp=<3500 3500 3000>
    componentLinearAccScale=<0.5 0.5 0.5>
    componentLinearVelScale=<0.65 0.65 0.65>
    gravityOverride {z=-980}
    physicsBodyDefinitions [{
      collisionType=CustomSphere
      sphereCollisionRadius=2
      boundBone {boneName=dyn_Hair_BackL_03}
      boxExtents=<25 25 25>
      constraintSetup {angularLimitsMax=<0 0 20>, angularLimitsMin=<0 0 -7>, angularTarget=<0 0 0>}
      localJointOffset=<0 -5 0>
     }]
    relativeSpaceBone {boneName=root}
   }
   n57 = Anim Dynamics - Bone: dyn_Hair_BackL_04(ComponentPose=@n56, Alpha=1, ExternalForce="0,0,0") {
    angularDampingOverride=0.99
    angularSpringConstant=75
    bAngularSpring=true
    bGravityOverrideInSimSpace=true
    bOverrideAngularDamping=true
    bUseGravityOverride=true
    bUseSphericalLimits=true
    simulationSpace=RootRelative
    alphaBoolBlend {blendOption=Linear}
    boundBone {boneName=dyn_Hair_BackL_04}
    componentAppliedLinearAccClamp=<3500 3500 3000>
    componentLinearAccScale=<0.5 0.5 0.5>
    componentLinearVelScale=<0.65 0.65 0.65>
    gravityOverride {z=-980}
    physicsBodyDefinitions [{
      collisionType=CustomSphere
      sphereCollisionRadius=2
      boundBone {boneName=dyn_Hair_BackL_04}
      boxExtents=<23 23 23>
      constraintSetup {angularLimitsMax=<15 2 30>, angularLimitsMin=<-5 -2 -7>, angularTarget=<0 0 0>}
      localJointOffset=<0 -5 0>
     }]
    relativeSpaceBone {boneName=root}
    sphericalLimits [
     {limitRadius=16, drivingBone {boneName=spine_05}, sphereLocalOffset=<0 0 0>}
     {limitRadius=0, drivingBone {boneName=None}, sphereLocalOffset=<0 0 0>}
    ]
   }
   n58 = Anim Dynamics - Bone: dyn_Hair_BackL_05(ComponentPose=@n57, Alpha=1, ExternalForce="0,0,0") {
    angularDampingOverride=0.99
    angularSpringConstant=75
    bAngularSpring=true
    bGravityOverrideInSimSpace=true
    bOverrideAngularDamping=true
    bUseGravityOverride=true
    bUseSphericalLimits=true
    simulationSpace=RootRelative
    alphaBoolBlend {blendOption=Linear}
    boundBone {boneName=dyn_Hair_BackL_05}
    componentAppliedLinearAccClamp=<3500 3500 3000>
    componentLinearAccScale=<0.5 0.5 0.5>
    componentLinearVelScale=<0.65 0.65 0.65>
    gravityOverride {z=-980}
    physicsBodyDefinitions [{
      collisionType=CustomSphere
      sphereCollisionRadius=2
      boundBone {boneName=dyn_Hair_BackL_05}
      boxExtents=<23 23 23>
      constraintSetup {angularLimitsMax=<15 2 30>, angularLimitsMin=<-5 -2 -10>, angularTarget=<0 0 0>}
      localJointOffset=<0 -5 0>
     }]
    relativeSpaceBone {boneName=root}
    sphericalLimits [
     {limitRadius=16, drivingBone {boneName=spine_05}, sphereLocalOffset=<0 0 0>}
     {limitRadius=0, drivingBone {boneName=None}, sphereLocalOffset=<0 0 0>}
    ]
   }
   n59 = Anim Dynamics - Bone: dyn_Hair_Back_02(ComponentPose=@n58, Alpha=0.2, ExternalForce="0,0,0") {
    angularDampingOverride=0.99
    angularSpringConstant=75
    bAngularSpring=true
    bGravityOverrideInSimSpace=true
    bOverrideAngularDamping=true
    bUseGravityOverride=true
    simulationSpace=RootRelative
    alphaBoolBlend {blendOption=Linear}
    boundBone {boneName=dyn_Hair_Back_02}
    componentAppliedLinearAccClamp=<3500 3500 3000>
    componentLinearAccScale=<0.5 0.5 0.5>
    componentLinearVelScale=<0.65 0.65 0.65>
    gravityOverride {z=-980}
    physicsBodyDefinitions [{
      collisionType=CustomSphere
      sphereCollisionRadius=2
      boundBone {boneName=dyn_Hair_Back_02}
      boxExtents=<40 40 40>
      constraintSetup {angularLimitsMax=<0 0 5>, angularLimitsMin=<0 0 -3>, angularTarget=<0 0 0>}
      localJointOffset=<-2 -5 0>
     }]
    relativeSpaceBone {boneName=root}
   }
   n60 = Anim Dynamics - Bone: dyn_Hair_Back_03(ComponentPose=@n59, Alpha=1, ExternalForce="0,0,0") {
    angularDampingOverride=0.99
    angularSpringConstant=75
    bAngularSpring=true
    bGravityOverrideInSimSpace=true
    bOverrideAngularDamping=true
    bUseGravityOverride=true
    simulationSpace=RootRelative
    alphaBoolBlend {blendOption=Linear}
    boundBone {boneName=dyn_Hair_Back_03}
    componentAppliedLinearAccClamp=<3500 3500 3000>
    componentLinearAccScale=<0.5 0.5 0.5>
    componentLinearVelScale=<0.65 0.65 0.65>
    gravityOverride {z=-980}
    physicsBodyDefinitions [{
      collisionType=CustomSphere
      sphereCollisionRadius=2
      boundBone {boneName=dyn_Hair_Back_03}
      boxExtents=<25 25 25>
      constraintSetup {angularLimitsMax=<0 0 20>, angularLimitsMin=<0 0 -7>, angularTarget=<0 0 0>}
      localJointOffset=<0 -5 0>
     }]
    relativeSpaceBone {boneName=root}
   }
   n61 = Anim Dynamics - Bone: dyn_Hair_Back_04(ComponentPose=@n60, Alpha=1, ExternalForce="0,0,0") {
    angularDampingOverride=0.99
    angularSpringConstant=75
    bAngularSpring=true
    bGravityOverrideInSimSpace=true
    bOverrideAngularDamping=true
    bUseGravityOverride=true
    bUseSphericalLimits=true
    simulationSpace=RootRelative
    alphaBoolBlend {blendOption=Linear}
    boundBone {boneName=dyn_Hair_Back_04}
    componentAppliedLinearAccClamp=<3500 3500 3000>
    componentLinearAccScale=<0.5 0.5 0.5>
    componentLinearVelScale=<0.65 0.65 0.65>
    gravityOverride {z=-980}
    physicsBodyDefinitions [{
      collisionType=CustomSphere
      sphereCollisionRadius=2
      boundBone {boneName=dyn_Hair_Back_04}
      boxExtents=<23 23 23>
      constraintSetup {angularLimitsMax=<2 2 30>, angularLimitsMin=<-2 -2 -7>, angularTarget=<0 0 0>}
      localJointOffset=<0 -5 0>
     }]
    relativeSpaceBone {boneName=root}
    sphericalLimits [{limitRadius=15, drivingBone {boneName=spine_05}, sphereLocalOffset=<0 0 0>}]
   }
   n62 = Anim Dynamics - Bone: dyn_Hair_Back_05(ComponentPose=@n61, Alpha=1, ExternalForce="0,0,0") {
    angularDampingOverride=0.99
    angularSpringConstant=75
    bAngularSpring=true
    bGravityOverrideInSimSpace=true
    bOverrideAngularDamping=true
    bUseGravityOverride=true
    bUseSphericalLimits=true
    simulationSpace=RootRelative
    alphaBoolBlend {blendOption=Linear}
    boundBone {boneName=dyn_Hair_Back_05}
    componentAppliedLinearAccClamp=<3500 3500 3000>
    componentLinearAccScale=<0.5 0.5 0.5>
    componentLinearVelScale=<0.65 0.65 0.65>
    gravityOverride {z=-980}
    physicsBodyDefinitions [{
      collisionType=CustomSphere
      sphereCollisionRadius=2
      boundBone {boneName=dyn_Hair_Back_05}
      boxExtents=<23 23 23>
      constraintSetup {angularLimitsMax=<3 2 30>, angularLimitsMin=<-3 -2 -10>, angularTarget=<0 0 0>}
      localJointOffset=<0 -5 0>
     }]
    relativeSpaceBone {boneName=root}
    sphericalLimits [
     {limitRadius=16, drivingBone {boneName=spine_05}, sphereLocalOffset=<0 0 0>}
     {limitRadius=0, drivingBone {boneName=None}, sphereLocalOffset=<0 0 0>}
    ]
   }
   n63 = Anim Dynamics - Bone: dyn_Hair_Back_06(ComponentPose=@n62, Alpha=1, ExternalForce="0,0,0") {
    angularDampingOverride=0.99
    angularSpringConstant=75
    bAngularSpring=true
    bGravityOverrideInSimSpace=true
    bOverrideAngularDamping=true
    bUseGravityOverride=true
    bUseSphericalLimits=true
    simulationSpace=RootRelative
    alphaBoolBlend {blendOption=Linear}
    boundBone {boneName=dyn_Hair_Back_06}
    componentAppliedLinearAccClamp=<3500 3500 3000>
    componentLinearAccScale=<0.5 0.5 0.5>
    componentLinearVelScale=<0.65 0.65 0.65>
    gravityOverride {z=-980}
    physicsBodyDefinitions [{
      collisionType=CustomSphere
      sphereCollisionRadius=2
      boundBone {boneName=dyn_Hair_Back_06}
      boxExtents=<23 23 23>
      constraintSetup {angularLimitsMax=<4 2 30>, angularLimitsMin=<-4 -2 -15>, angularTarget=<0 0 0>}
      localJointOffset=<0 -5 0>
     }]
    relativeSpaceBone {boneName=root}
    sphericalLimits [
     {limitRadius=16, drivingBone {boneName=spine_05}, sphereLocalOffset=<0 0 0>}
     {limitRadius=0, drivingBone {boneName=None}, sphereLocalOffset=<0 0 0>}
    ]
   }
   n64 = Anim Dynamics - Bone: dyn_Hair_BackR_02(ComponentPose=@n63, Alpha=0.2, ExternalForce="0,0,0") {
    angularDampingOverride=0.99
    angularSpringConstant=75
    bAngularSpring=true
    bGravityOverrideInSimSpace=true
    bOverrideAngularDamping=true
    bUseGravityOverride=true
    simulationSpace=RootRelative
    alphaBoolBlend {blendOption=Linear}
    boundBone {boneName=dyn_Hair_BackR_02}
    componentAppliedLinearAccClamp=<3500 3500 3000>
    componentLinearAccScale=<0.5 0.5 0.5>
    componentLinearVelScale=<0.65 0.65 0.65>
    gravityOverride {z=-980}
    physicsBodyDefinitions [{
      collisionType=CustomSphere
      sphereCollisionRadius=2
      boundBone {boneName=dyn_Hair_BackR_02}
      boxExtents=<40 40 40>
      constraintSetup {angularLimitsMax=<0 0 5>, angularLimitsMin=<0 0 -3>, angularTarget=<0 0 0>}
      localJointOffset=<-2 -5 0>
     }]
    relativeSpaceBone {boneName=root}
   }
   n65 = Anim Dynamics - Bone: dyn_Hair_BackR_03(ComponentPose=@n64, Alpha=1, ExternalForce="0,0,0") {
    angularDampingOverride=0.99
    angularSpringConstant=75
    bAngularSpring=true
    bGravityOverrideInSimSpace=true
    bOverrideAngularDamping=true
    bUseGravityOverride=true
    simulationSpace=RootRelative
    alphaBoolBlend {blendOption=Linear}
    boundBone {boneName=dyn_Hair_BackR_03}
    componentAppliedLinearAccClamp=<3500 3500 3000>
    componentLinearAccScale=<0.5 0.5 0.5>
    componentLinearVelScale=<0.65 0.65 0.65>
    gravityOverride {z=-980}
    physicsBodyDefinitions [{
      collisionType=CustomSphere
      sphereCollisionRadius=2
      boundBone {boneName=dyn_Hair_BackR_03}
      boxExtents=<25 25 25>
      constraintSetup {angularLimitsMax=<0 0 20>, angularLimitsMin=<0 0 -7>, angularTarget=<0 0 0>}
      localJointOffset=<0 -5 0>
     }]
    relativeSpaceBone {boneName=root}
   }
   n66 = Anim Dynamics - Bone: dyn_Hair_BackR_04(ComponentPose=@n65, Alpha=1, ExternalForce="0,0,0") {
    angularDampingOverride=0.99
    angularSpringConstant=75
    bAngularSpring=true
    bGravityOverrideInSimSpace=true
    bOverrideAngularDamping=true
    bUseGravityOverride=true
    bUseSphericalLimits=true
    simulationSpace=RootRelative
    alphaBoolBlend {blendOption=Linear}
    boundBone {boneName=dyn_Hair_BackR_04}
    componentAppliedLinearAccClamp=<3500 3500 3000>
    componentLinearAccScale=<0.5 0.5 0.5>
    componentLinearVelScale=<0.65 0.65 0.65>
    gravityOverride {z=-980}
    physicsBodyDefinitions [{
      collisionType=CustomSphere
      sphereCollisionRadius=2
      boundBone {boneName=dyn_Hair_BackR_04}
      boxExtents=<23 23 23>
      constraintSetup {angularLimitsMax=<5 2 30>, angularLimitsMin=<-15 -2 -7>, angularTarget=<0 0 0>}
      localJointOffset=<0 -5 0>
     }]
    relativeSpaceBone {boneName=root}
    sphericalLimits [
     {limitRadius=16, drivingBone {boneName=spine_05}, sphereLocalOffset=<0 0 0>}
     {limitRadius=0, drivingBone {boneName=None}, sphereLocalOffset=<0 0 0>}
    ]
   }
   n67 = Anim Dynamics - Bone: dyn_Hair_BackR_05(ComponentPose=@n66, Alpha=1, ExternalForce="0,0,0") {
    angularDampingOverride=0.99
    angularSpringConstant=75
    bAngularSpring=true
    bGravityOverrideInSimSpace=true
    bOverrideAngularDamping=true
    bUseGravityOverride=true
    bUseSphericalLimits=true
    simulationSpace=RootRelative
    alphaBoolBlend {blendOption=Linear}
    boundBone {boneName=dyn_Hair_BackR_05}
    componentAppliedLinearAccClamp=<3500 3500 3000>
    componentLinearAccScale=<0.5 0.5 0.5>
    componentLinearVelScale=<0.65 0.65 0.65>
    gravityOverride {z=-980}
    physicsBodyDefinitions [{
      collisionType=CustomSphere
      sphereCollisionRadius=2
      boundBone {boneName=dyn_Hair_BackR_05}
      boxExtents=<23 23 23>
      constraintSetup {angularLimitsMax=<5 0 30>, angularLimitsMin=<-15 0 -10>, angularTarget=<0 0 0>}
      localJointOffset=<0 -5 0>
     }]
    relativeSpaceBone {boneName=root}
    sphericalLimits [
     {limitRadius=16, drivingBone {boneName=spine_05}, sphereLocalOffset=<0 0 0>}
     {limitRadius=0, drivingBone {boneName=None}, sphereLocalOffset=<0 0 0>}
    ]
   }
   Save cached pose 'HairDynPose'(Pose=Component To Local(ComponentPose=@n67)) {cacheName=HairDynPose}
   n68 = Get alpha constraint 03() -> (alpha constraint 03:double)
   n69 = "Constraint - dyn_hair_l_03 (PP)"(Alpha=@n68, ConstraintWeights_0=0.8, ConstraintWeights_1=0.2) {
    alphaBoolBlend {blendOption=Linear}
    alphaScaleBiasClamp {bInterpResult=true, interpSpeedDecreasing=1}
    boneToModify {boneName=dyn_hair_l_03}
    constraintSetup [{targetBone {boneName=spine_05}}, {targetBone {boneName=clavicle_l}}]
    constraintWeights [1,1]
   }
   n70 = "Constraint - dyn_hair_middle_03 (P)"(ComponentPose=@n69, Alpha=@n68, ConstraintWeights_0=1) {
    alphaBoolBlend {blendOption=Linear}
    alphaScaleBiasClamp {bInterpResult=true, interpSpeedDecreasing=1}
    boneToModify {boneName=dyn_hair_middle_03}
    constraintSetup [{targetBone {boneName=spine_05}}]
    constraintWeights [1]
   }
   n71 = "Constraint - dyn_hair_middle_04 (P)"(ComponentPose=@n70, Alpha=@n68, ConstraintWeights_0=1) {
    alphaBoolBlend {blendOption=Linear}
    alphaScaleBiasClamp {bInterpResult=true, interpSpeedDecreasing=1}
    boneToModify {boneName=dyn_hair_middle_04}
    constraintSetup [{targetBone {boneName=spine_05}}]
    constraintWeights [1]
   }
   n72 = "Constraint - dyn_hair_middle_05 (P)"(ComponentPose=@n71, Alpha=@n68, ConstraintWeights_0=1) {
    alphaBoolBlend {blendOption=Linear}
    alphaScaleBiasClamp {bInterpResult=true, interpSpeedDecreasing=1}
    boneToModify {boneName=dyn_hair_middle_05}
    constraintSetup [{targetBone {boneName=spine_05}}]
    constraintWeights [1]
   }
   n73 = "Constraint - dyn_hair_r_01 (P)"(ComponentPose=@n72, Alpha=@n68, ConstraintWeights_0=1) {
    alphaBoolBlend {blendOption=Linear}
    alphaScaleBiasClamp {bInterpResult=true, interpSpeedDecreasing=1}
    boneToModify {boneName=dyn_hair_r_01}
    constraintSetup [{targetBone {boneName=spine_05}}]
    constraintWeights [1]
   }
   n74 = "Constraint - dyn_hair_r_02 (PP)"!(ComponentPose=@n73, Alpha=@n68, ConstraintWeights_0=1, ConstraintWeights_1:float=0) -> (Pose) {
    alphaBoolBlend {blendOption=Linear}
    alphaScaleBiasClamp {bInterpResult=true, interpSpeedDecreasing=1}
    boneToModify {boneName=dyn_hair_r_02}
    constraintSetup [{targetBone {boneName=spine_05}}, {targetBone {boneName=clavicle_r}}]
    constraintWeights [1,1]
   }
   n75 = "Constraint - dyn_hair_r_03 (PPP)"!(ComponentPose=@n74, Alpha=0.5, ConstraintWeights_0=1, ConstraintWeights_1:float=0, ConstraintWeights_2:float=0) -> (Pose) {
    alphaBoolBlend {blendOption=Linear}
    alphaScaleBiasClamp {bInterpResult=true, interpSpeedDecreasing=1}
    boneToModify {boneName=dyn_hair_r_03}
    constraintSetup [{targetBone {boneName=spine_05}}, {targetBone {boneName=clavicle_r}}, {targetBone {boneName=dyn_hair_middle_04}}]
    constraintWeights [1,1,1]
   }
   n76 = "Constraint - dyn_hair_fringe_l_09 (P)"(ComponentPose=@n75, Alpha=@n68, ConstraintWeights_0=1) {
    alphaBoolBlend {blendOption=Linear}
    alphaScaleBiasClamp {bInterpResult=true, interpSpeedDecreasing=1}
    boneToModify {boneName=dyn_hair_fringe_l_09}
    constraintSetup [{targetBone {boneName=spine_05}}]
    constraintWeights [1]
   }
   Reroute Node(InputPin:Component Space Pose Link Structure=@n76) -> (OutputPin:Component Space Pose Link Structure)
  ]}
 DefineHair17Transform {nodes [
   DefineHair17Transform() {is=title K2Node_FunctionEntry}
   n1 = Select!(
    Option 0:Transform=Select!(
     Option 0=MakeTransform(Location="-4,0,0", Rotation="0,30,0")
     Option 1=MakeTransform(Location="-5,1,0", Rotation="10,45,0")
     Index=Get bIsTargeting()
    ) -> (ReturnValue:Transform)
    Option 1:Transform
    Index=Get bIsCrouching()
   ) -> (ReturnValue_Location:Vector, ReturnValue_Rotation:Rotator, ReturnValue_Scale:Vector)
   > Set Translation(Translation:Vector=@n1) -> (Output_Get:Vector)
   > Set Rotation(Rotation:Rotator=@n1.ReturnValue_Rotation) -> (Output_Get:Rotator)
   MakeTransform(Location="-5,0,0", Rotation="0,35,0")
   n2 = Get WeaponCoreAnim!() -> (WeaponCoreAnim)
   "Equal (Enum)"(A:EFortWeaponCoreAnimation Enum=@n2, B:EFortWeaponCoreAnimation Enum=Pistol)
   "Not Equal (Enum)"!(A:EFortWeaponCoreAnimation Enum=@n2, B:EFortWeaponCoreAnimation Enum=Pistol) -> (ReturnValue:bool)
   Get bIsRelaxedLevel1() -> (bIsRelaxedLevel1:bool)
  ]}
 EventGraph {nodes [
   n1 = Event BlueprintUpdateAnimation()
   > Set delta time x(delta time x:float=@n1.DeltaTimeX) -> (Output_Get:float)
   > n2 = Sequence!(execute) -> (then_0, then_1, then_2:Exec)
   > n3 = Gravity Override State Function DEPRECATED(CharacterPartAnimInstance=Self-Reference(), GameTimeInSeconds=GetGameTimeInSeconds(), GravityOverrideParams=Get Gravity Override Params())
   > Set Gravity Override(Gravity Override=@n3)
   Set is crouching float(execute=@n2.then_1, is crouching float:double=SelectFloat(A=0.5, B=0, bPickA=Get bIsCrouching())) -> (Output_Get:double)
   n4 = Get bIsCrouchSprinting()
   n5 = Get bIsPlayingMeleeAnim()
   > Branch(Condition=AND Boolean(A=NOT Boolean(A=@n5), B=NOT Boolean(A=@n4)))
   > Set head up straight(head up straight:float=GetAlphaFromReferenceTransformDelta(SkeletalMeshComponent=GetOwnerSkeletalMeshComponent(), BoneName=head, BaseName=spine_05, InputRange=(X=-19,Y=-40), TwistAxis="1,0,0")) -> (Output_Get:float)
   > Set head to the shoulder(head to the shoulder:float=GetAlphaFromReferenceTransformDelta(
     SkeletalMeshComponent=GetOwnerSkeletalMeshComponent()
     BoneName=head
     BaseName=spine_05
     DeltaComponentType=EulerY
     InputRange=(X=10,Y=22)
     TwistAxis="1,0,0"
    )) -> (Output_Get:float)
   n6 = float + float(A=Get head up straight() -> (head up straight:float), B=Get head to the shoulder() -> (head to the shoulder:float))
   > Set head up or to the shoulder(head up or to the shoulder:double="Clamp (Float)"(Value=float / float(A=@n6, B=1.3), Min=0, Max=1)) -> (Output_Get:double)
   n7 = "To Float (Boolean)"(InBool=NOT Boolean(A=Get bIsSkyDiving()))
   n8 = GetAlphaFromReferenceTransformDelta(SkeletalMeshComponent=GetOwnerSkeletalMeshComponent(), BoneName=head, BaseName=spine_05, InputRange=(X=-19,Y=-40), TwistAxis="1,0,0")
   n9 = GetAlphaFromReferenceTransformDelta(
    SkeletalMeshComponent=GetOwnerSkeletalMeshComponent()
    BoneName=head
    BaseName=spine_05
    DeltaComponentType=EulerY
    InputRange=(X=10,Y=22)
    TwistAxis="1,0,0"
   )
   > Set alpha constraint 03(alpha constraint 03:double="Clamp (Float)"(Value=float + float!(A=@n8, B=@n9, C:double=@n7) -> (ReturnValue), Min=0, Max=0.3)) -> (Output_Get:double)
   n10 = Get bIsPlayingMeleeAnim()
   n11 = Get bIsTacticalSprinting()
   Set mele and tactical(execute=@n2.then_2, mele and tactical:double=SelectFloat(A=0.2, B=0, bPickA=OR Boolean!(A=@n10, B=@n11) -> (ReturnValue))) -> (Output_Get:double)
   > n12 = Branch(Condition=Get bIsMoving2D())
   > Set look at alpha(look at alpha:double=0.5) -> (Output_Get:double)
   n13 = Branch(execute=@n12.else, Condition=Get bIsPlayingMeleeAnim())
   > Set look at alpha(look at alpha=1)
   Set look at alpha(execute=@n13.else, look at alpha=0)
   Event OnCharacterCustomizationLoaded()
   n14 = CheckForBackbling()
   > Set isWearingPackBool(isWearingPackBool:bool=OR Boolean!(A=@n14, B=@n14.bIsBackblingACape, C:bool=@n14.bIsBackblingLong) -> (ReturnValue)) -> (Output_Get:bool)
   AND Boolean(A=@n14, B=NOT Boolean(A=@n14.bIsBackblingLong))
  ]}
]
variables [
 alpha constraint 03 {type=number}
 delta time x {type=number}
 Gravity Override {type=Vector}
 Gravity Override Params {
  type=GravityOverrideParamsStruct
  default {
   skydive_DivingDown {windAmplitude=<50 150 100>}
   skydive_FloatingDown {gravityOverride=<0 0 0>, windAmplitude=<20 -20 50>, windOffset=<0 -300 100>}
   skydive_GoingUp {gravityOverride=<0 0 0>, windAmplitude=<20 -20 50>, windOffset=<0 -300 100>}
  }
 }
 head to the shoulder {type=number}
 head up or to the shoulder {type=number}
 head up straight {type=number}
 is crouching float {type=number}
 isWearingPackBool {type=boolean}
 look at alpha {type=number}
 mele and tactical {type=number}
 Rotation {type=Rotator, default=<rot 0 35 0>}
 Translation {type=Vector, default=<-5 0 0>}
]
